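# Parent material function for the CanineCronutDig character skin, adding glittery clear coat and audio-reactive glow to incoming material attributes.
# It combines optional eye, hair and desaturation/sepia stages with glitter, sky-light fake reflections and reactive fresnel colour.
# Front-end loop amplitude, tech-art masks and a pre-skinned sphere mask feed the effects through named reroutes.
# Quality switches scale sparkle per platform, and several branches, including a world-position-offset shell, are left unconnected.
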MaterialFunction /BRCosmetics/Characters/Player/Male/Medium/Bodies/M_MED_CanineCronutDig/Materials/MF_CanineCronutDig_Ware_Parent
graph {nodes [
  n1 = GetMaterialAttributes(None=NamedRerouteUsage()) {attributeGetTypes [#1]}
  n2 = Multiply(A=Multiply(A=NamedRerouteUsage(), B=NamedRerouteUsage()), B=NamedRerouteUsage())
  FunctionOutput(None=SetMaterialAttributes!(MaterialAttributes=@n1.MaterialAttributes, EmissiveColor=Max(A=@n1.EmissiveColor, B=@n2)) {attributeSetTypes [#1]}) {outputName=MA}
  MaterialFunctionCall() {materialFunction=/Game/Packages/Fortress_SharedMaterials/Base_Material_Functions/BaseCharacterShaderFunctions/Audio/MF_AudioReactivity}
  n3 = NamedRerouteUsage()
  n4 = Saturate(None=NamedRerouteUsage())
  Multiply(A=@n3, B=@n4)
  NamedRerouteDeclaration(None=FunctionInput() {inputName=Result, inputType=FunctionInput_MaterialAttributes}) {name=Material Attribute - In, nodeColor=<rgba 0.4531 1 0 1>}
  NamedRerouteDeclaration(None=FunctionInput() {inputName=M, inputType=FunctionInput_Scalar, sortPriority=10}) {name=Masks - In, nodeColor=<rgba 1 0 0.7734 1>}
  n5 = Max(A=NamedRerouteUsage(), B=NamedRerouteUsage())
  NamedRerouteDeclaration(None=@n5) {name=Combined Mask, nodeColor=<rgba 0.7109 1 0 1>}
  n6 = LocalPosition() {localOrigin=InstancePreSkinning}
  NamedRerouteDeclaration(None=SphereMask(
    A=VertexInterpolator(VS=@n6.XYZ).PS
    B=VectorParameter() {group=CanineCronutDig_Ware, parameterName=Reactive_MaskCenter, defaultValue=<rgba 0 0 141.908 1>}.RGB
    Radius=ScalarParameter() {defaultValue=120, group=CanineCronutDig_Ware, parameterName=Reactive_Radius}
    Hardness=ScalarParameter() {defaultValue=0.2, group=CanineCronutDig_Ware, parameterName=Reactive_Hardness}
   )) {name=Sphere Mask, nodeColor=<rgba 0 1 0.1562 1>}
  n7 = TextureSample(UVs=Multiply(A=TextureCoordinate(), B=ScalarParameter() {defaultValue=25, group=CanineCronutDig_Ware, parameterName=Random Glitter - Tiling})) {samplerType=SAMPLERTYPE_Masks, texture=/Game/Weapons/WeaponSkins/Textures/T_GlitterDirectionMask}
  n8 = Multiply(
   A=Add(A=Multiply(A=@n7.RGB) {constB=2}) {constB=-1}
   B=ScalarParameter() {defaultValue=0.1, group=CanineCronutDig_Ware, parameterName=Random Glitter - Strength, sliderMax=2, sortPriority=6}
  ) {constB=0.5}
  n9 = Normalize(VectorInput=Add(A=NamedRerouteDeclaration(None=@n8) {name=Random Normal Dir, nodeColor=<rgba 0 0.3501 1 1>}, B=VertexNormalWS()))
  NamedRerouteDeclaration(None=Transform(None=@n9) {transformSourceType=TRANSFORMSOURCE_World, transformType=TRANSFORM_Tangent}) {name=Sparkle Normal, nodeColor=<rgba 0.7812 1 0 1>}
  n10 = StaticSwitchParameter(True=MaterialFunctionCall() {materialFunction=/Game/Characters/Shared/FNEye/Materials/MF_eyes}.Result, False=NamedRerouteUsage()) {defaultValue=true, group=CanineCronutDig_Ware, parameterName=Use Eyes?}
  n11 = GetMaterialAttributes(None=@n10) {attributeGetTypes [#2, #1]}
  n12 = Desaturation(None=@n11.BaseColor, Fraction=ScalarParameter() {group=CanineCronutDig_Ware, parameterName=Desaturation})
  n13 = ScalarParameter() {defaultValue=1, group=CanineCronutDig_Ware, parameterName=Desaturation Pow}
  n14 = Multiply(
   A=Multiply(
    A=Power(Base=@n12, Exp=@n13) {constExponent=1}
    B=ScalarParameter() {defaultValue=1, group=CanineCronutDig_Ware, parameterName=Desaturation Intensity}
   )
   B=VectorParameter() {group=CanineCronutDig_Ware, parameterName=Sepia Color, defaultValue=<rgba 0.8333 1 0.8478 1>}.RGB
  )
  n15 = Desaturation(None=@n11.EmissiveColor, Fraction=ScalarParameter() {group=CanineCronutDig_Ware, parameterName=Desaturation})
  n16 = ScalarParameter() {defaultValue=1, group=CanineCronutDig_Ware, parameterName=Desaturation Pow}
  n17 = Multiply(
   A=Multiply(
    A=Power(Base=@n15, Exp=@n16) {constExponent=1}
    B=ScalarParameter() {defaultValue=1, group=CanineCronutDig_Ware, parameterName=Desaturation Intensity}
   )
   B=VectorParameter() {group=CanineCronutDig_Ware, parameterName=Sepia Color, defaultValue=<rgba 0.8333 1 0.8478 1>}.RGB
  )
  NamedRerouteDeclaration(None=StaticSwitchParameter(
    True=SetMaterialAttributes!(MaterialAttributes=@n11.MaterialAttributes, BaseColor=@n14, EmissiveColor=@n17) {attributeSetTypes [#2, #1]}
    False=@n10
   ) {group=CanineCronutDig_Ware, parameterName=Desaturation}) {name=Desaturated Base Color, nodeColor=<rgba 0.7344 0 1 1>}
  n18 = Power(
   Base=DotProduct(A=@n9, B=CameraVectorWS())
   Exp=ScalarParameter() {defaultValue=1, group=CanineCronutDig_Ware, parameterName=Random Glitter - Exp}
  )
  NamedRerouteDeclaration(None=Multiply(A=Saturate(None=@n18), B=NamedRerouteUsage())) {name=Random Glitter, nodeColor=<rgba 0 1 0.4141 1>}
  n19 = MaterialFunctionCall!(Mask=NamedRerouteUsage(), Fresnel=NamedRerouteUsage(), In=NamedRerouteUsage()) {materialFunction=/Game/Global/MaterialFunctions/TechArt/MF_Advanced_CarClearCoat_v2}
  n20 = GetMaterialAttributes(None=@n19.Result) {attributeGetTypes [#3, #4, #5]}
  n21 = NamedRerouteUsage()
  n22 = SetMaterialAttributes!(
   MaterialAttributes=@n20.MaterialAttributes
   Roughness=LinearInterpolate(
    A=@n20.Roughness
    B=ScalarParameter() {defaultValue=0.8171, group=CanineCronutDig_Ware, parameterName=Random Glitter - Roughness}
    Alpha=@n21
   )
   Normal=LinearInterpolate(
    A=@n20.Normal
    B=MaterialFunctionCall(BaseNormal=@n20.Normal, AdditionalNormal=@n19.ClearCoatNormal) {materialFunction=&BlendAngleCorrectedNormals}.Result
    Alpha=@n21
   )
   ShadingModel=ShadingModel()
   Specular
   EmissiveColor
  ) {attributeSetTypes [
    #3
    #4
    #6
    #5
    #1
   ]}
  n23 = GetMaterialAttributes(None=MaterialFunctionCall!(In=@n22, SkinFX_Mask=@n21) {materialFunction=/BRCosmetics/Characters/Player/Female/Medium/Bodies/F_MED_HangSpec_Screech/Skins/MF_GlitterFX}.Result) {attributeGetTypes [#1, #2]}
  n24 = Multiply(A=@n23.BaseColor, B=Desaturation(None=@n23.EmissiveColor))
  n25 = VectorParameter() {parameterName=Sepia Sparkle Overtone, defaultValue=<rgba 1 0.9512 0.901 1>}
  n26 = Multiply(
   A=Multiply(A=Multiply(A=@n24, B=@n25.RGB), B=NamedRerouteUsage())
   B=ScalarParameter() {defaultValue=15, group=CanineCronutDig_Ware, parameterName=Glitter - Ambient Brightness}
  )
  n27 = Ceil(None=NamedRerouteUsage())
  NamedRerouteDeclaration(None=SetMaterialAttributes!(MaterialAttributes=@n23.MaterialAttributes, EmissiveColor=LinearInterpolate(A=@n26, B=@n23.EmissiveColor, Alpha=@n27)) {attributeSetTypes [#1]}) {name=Glitter FX, nodeColor=<rgba 1 0.0938 0 1>}
  n28 = ScalarParameter() {defaultValue=0.9, group=CanineCronutDig_Ware, parameterName=Glitter - Transition Amount}
  n29 = ScalarParameter() {defaultValue=0.15, group=CanineCronutDig_Ware, parameterName=Glitter - Transition Fade}
  n30 = Normalize(VectorInput=ReflectionVectorWS())
  n31 = Normalize(VectorInput=Add(A=NamedRerouteUsage(), B=VertexNormalWS())) {desc=GlitterDirection}
  NamedRerouteDeclaration(None=SmoothStep(Min=@n28, Max=Add(A=@n28, B=@n29), Value=DotProduct(A=@n30, B=@n31))) {name=Reflection Glitter - Mask, nodeColor=<rgba 0 0.2266 1 1>}
  n32 = NamedRerouteUsage()
  n33 = SkyLightEnvMapSample()
  n34 = LinearInterpolate(
   A=Desaturation(None=@n33, Fraction=ScalarParameter() {defaultValue=1, group=CanineCronutDig_Ware, parameterName=Reflection Glitter - Desaturation})
   B=Multiply(
    A=VectorParameter() {group=CanineCronutDig_Ware, parameterName=Reactive Color, defaultValue=<rgba 0.0052 0.2232 0.0578 1>}.RGB
    B=ScalarParameter() {defaultValue=5, group=CanineCronutDig_Ware, parameterName=Reflection Glitter - Intensity}
   )
   Alpha=Ceil(None=NamedRerouteUsage())
  )
  n35 = ScalarParameter() {group=CanineCronutDig_Ware, parameterName=Global Fake Reflection - Intensity}
  n36 = ScalarParameter() {group=CanineCronutDig_Ware, parameterName=Global Fake Reflection Reactive - Intensity}
  n37 = Saturate(None=NamedRerouteUsage())
  NamedRerouteDeclaration(None=Multiply(A=Multiply(A=@n32, B=@n34), B=LinearInterpolate(A=@n35, B=@n36, Alpha=@n37))) {name=Reflection Glitter, nodeColor=<rgba 0.5469 1 0 1>}
  n38 = Divide(A=OneMinus(None=LocalPosition().Z), B=ScalarParameter() {defaultValue=149.8988, group=CanineCronutDig_Ware, parameterName=Sparkle - Fade})
  n39 = Divide(A=ComponentMask(None=PreSkinnedLocalBounds().Half Extents) {b=true}) {constB=100}
  NamedRerouteDeclaration(None=Saturate(None=Add(A=@n38, B=@n39))) {name=Glitter - Top Down Fade, nodeColor=<rgba 1 0 0.8672 1>}
  n40 = MaterialFunctionCall!(In=NamedRerouteUsage(), Mask Input=NamedRerouteUsage()) {materialFunction=/Game/Global/Materials/Character_Parent_2023/Hair_Parent_2023/MaterialFunctions/MF_FN_Hair_2023}
  n41 = BlendMaterialAttributes(
   A=NamedRerouteUsage()
   B=StaticSwitchParameter(True=@n40.Result, False=NamedRerouteUsage()) {defaultValue=true, group=Skin FX, parameterName=UseFNHair2_0}
   Alpha=@n40.HairMask
  )
  NamedRerouteDeclaration(None=StaticSwitchParameter(
    True=MaterialFunctionCall!(MaterialInput=@n41) {materialFunction=/Game/Global/Materials/Character_Parent_2023/Hair_Parent_2023/MaterialFunctions/MF_FlyAwayHair_2023}.Result
    False=@n41
   ) {group=Skin FX, parameterName=IsFlyAwayHair?}) {name=Hair + Original MA, nodeColor=<rgba 0 1 0.9062 1>}
  n42 = GetMaterialAttributes(None=NamedRerouteUsage()) {attributeGetTypes [#1]}
  n43 = NamedRerouteUsage()
  n44 = Desaturation(None=@n42.EmissiveColor)
  n45 = ScalarParameter() {defaultValue=4.5, group=CanineCronutDig_Ware, parameterName=Reactive Sparkle - Exp - Default}
  n46 = MaterialFunctionCall!(
   Low _ PC Low and Med=ScalarParameter() {defaultValue=1, group=CanineCronutDig_Ware, parameterName=Reactive Sparkle - Exp - PC Low}
   "High _ XB1, PS4"=@n45
   "Epic _ PC High, XSX, PS5"=@n45
   --------
   "Mobile Low _ PC Perf, Android"=ScalarParameter() {defaultValue=1, group=CanineCronutDig_Ware, parameterName=Reactive Sparkle - Exp - PC Perf}
   Mobile Med _ Switch=@n45
  ) {materialFunction=/Game/Global/MaterialFunctions/TechArt/MF_Fortnite_QualitySwitch_Scalar}
  n47 = NamedRerouteUsage()
  n48 = LinearInterpolate(
   A=ScalarParameter() {defaultValue=0.1, group=CanineCronutDig_Ware, parameterName=Reactive Sparkle - Min}
   B=ScalarParameter() {defaultValue=1000000, group=CanineCronutDig_Ware, parameterName=Reactive Sparkle - Max}
   Alpha=@n47
  )
  n49 = ScalarParameter() {defaultValue=0.001, group=Audio Reactivity, parameterName=Sparkle - PC Low - Multiplier}
  n50 = ScalarParameter() {defaultValue=0.0025, group=Audio Reactivity, parameterName=Sparkle - Perf Mode - Multiplier}
  n51 = Multiply(
   A=Multiply(A=@n42.EmissiveColor, B=Power(Base=@n44, Exp=@n46.Result))
   B=MaterialFunctionCall(
    Low _ PC Low and Med=Multiply(A=@n48, B=@n49)
    "High _ XB1, PS4"=@n48
    "Epic _ PC High, XSX, PS5"=@n48
    "Mobile Low _ PC Perf, Android"=Multiply(A=@n48, B=@n50)
    Mobile Med _ Switch=@n48
   ) {materialFunction=&MF_Fortnite_QualitySwitch_Scalar}.Result
  )
  n52 = NamedRerouteUsage()
  n53 = Constant() {r=1}
  n54 = LinearInterpolate(
   A=Multiply(A=Add(A=@n42.EmissiveColor, B=@n51), B=If(A=@n52, "A > B"=@n52, "A == B"=@n52, "A < B"=@n53) {constB=0.01})
   Alpha=NamedRerouteUsage()
  ) {constB=0}
  NamedRerouteDeclaration(None=SetMaterialAttributes(MaterialAttributes=@n42.MaterialAttributes, EmissiveColor=Max(A=@n43, B=@n54)) {attributeSetTypes [#1]}) {name=Reactive Sparkle, nodeColor=<rgba 0 1 0.8125 1>}
  n55 = Multiply(A=NamedRerouteUsage(), B=Ceil(None=NamedRerouteUsage()))
  n56 = NamedRerouteUsage()
  NamedRerouteDeclaration(None=LinearInterpolate(A=@n55, B=Multiply(A=@n55, B=@n56), Alpha=ScalarParameter() {group=CanineCronutDig_Ware, parameterName=Reactive Mask Override})) {name=Coloured Reactivity, nodeColor=<rgba 1 0.4922 0 1>}
  n57 = VectorParameter() {group=CanineCronutDig_Ware, parameterName=Reactive Color, defaultValue=<rgba 0.0052 0.2232 0.0578 1>}
  n58 = Multiply(
   A=@n57.RGB
   B=LinearInterpolate(
    A=ScalarParameter() {group=CanineCronutDig_Ware, parameterName=Reactive Color Boost}
    B=ScalarParameter() {defaultValue=10, group=CanineCronutDig_Ware, parameterName=Reactive Color Max}
    Alpha=NamedRerouteUsage()
   )
  )
  n59 = OneMinus(None=Fresnel(ExponentIn=ScalarParameter() {defaultValue=0.2, group=CanineCronutDig_Ware, parameterName=Reactive Fresnel Exp}) {baseReflectFraction=0})
  NamedRerouteDeclaration(None=Multiply(A=NamedRerouteUsage(), B=Multiply(A=@n58, B=@n59))) {name=Coloured Reactivity - Animation, nodeColor=<rgba 0.1719 0 1 1>}
  n60 = BlendMaterialAttributes(A=NamedRerouteUsage(), B=NamedRerouteUsage(), Alpha=NamedRerouteUsage())
  n61 = GetMaterialAttributes(None=@n60) {attributeGetTypes [#1, #2]}
  n62 = LinearInterpolate(A=@n61.EmissiveColor, B=NamedRerouteUsage(), Alpha=NamedRerouteUsage())
  n63 = LinearInterpolate(A=@n61.BaseColor, B=@n57.RGB, Alpha=NamedRerouteUsage()) {constB=0}
  NamedRerouteDeclaration(None=BlendMaterialAttributes(
    A=@n60
    B=SetMaterialAttributes!(MaterialAttributes=@n61.MaterialAttributes, EmissiveColor=@n62, BaseColor=@n63) {attributeSetTypes [#1, #2]}
    Alpha=NamedRerouteUsage()
   )) {name=Reactive Color, nodeColor=<rgba 0.1484 1 0 1>}
  NamedRerouteDeclaration(None=Power(Base=Desaturation(None=@n61.BaseColor), Exp=ScalarParameter() {defaultValue=0.6})) {name=Base Color Mask, nodeColor=<rgba 0.9453 1 0 1>}
  n64 = TextureCoordinate()
  n65 = Constant() {r=3}
  n66 = TextureSample(UVs=Panner(Coordinate=Multiply(A=@n64, B=@n65)) {speedX=0.1, speedY=0.1}) {samplerSource=SSM_Wrap_WorldGroupSettings, samplerType=SAMPLERTYPE_Grayscale, texture=&LowResBlurredNoise}
  NamedRerouteDeclaration(None=Power(Base=@n66.R, Exp=ScalarParameter() {defaultValue=2, parameterName=Noise Exp})) {name=ColorNoise, nodeColor=<rgba 0.1484 0 1 1>}
  n67 = Frac(None=Multiply(A=Time(), B=ScalarParameter() {defaultValue=0.5, group=Audio Reactivity, parameterName=FakeAudioCurveSpeed}))
  n68 = CurveAtlasRowParameter(CurveTime=@n67) {
   atlas=/BRCosmetics/Global/MaterialFunctions/AudioReactivity/Curves/CA_Frontend_FakeAudioReactivity
   curve=/BRCosmetics/Global/MaterialFunctions/AudioReactivity/Curves/C_Frontend_FakeAudioCurve_1
   group=Audio Reactivity
   parameterName=FakeAudioCurveAtlasIndex
  }
  n69 = AppendVector(A=@n68, B=@n68)
  n70 = ScalarParameter() {defaultValue=1, group=Audio Reactivity, parameterName=FakeAudioReactivityMaxIntensity}
  NamedRerouteDeclaration(None=ComponentMask(None=Multiply(A=@n69, B=@n70)) {r=true}) {name=FakeAudioData, nodeColor=<rgba 0.1719 0 1 1>}
  NamedRerouteDeclaration(None=@n33) {name=Sky Light Color, nodeColor=<rgba 0 1 0.9063 1>}
  n71 = FunctionInput() {inputName=Tech Art Mask, sortPriority=50}
  NamedRerouteDeclaration(None=Max(A=ScalarParameter() {group=CanineCronutDig_Ware, parameterName=Reactive Mask Override}, B=ComponentMask(None=@n71) {g=true})) {name=Reactive Mask, nodeColor=<rgba 0 0.5781 1 1>}
  NamedRerouteDeclaration(None=FunctionInput() {inputName=Fresnel, inputType=FunctionInput_Scalar, sortPriority=20}) {name=Fresnel - In, nodeColor=<rgba 0.4062 0 1 1>}
  NamedRerouteDeclaration(None=@n71) {name=TA Mask, nodeColor=<rgba 0 1 0.0859 1>}
  n72 = DotProduct(A=VertexNormalWS(), B=AtmosphericLightVector())
  NamedRerouteDeclaration(None=Multiply(A=Saturate(None=@n72), B=NamedRerouteUsage())) {name=LDotN, nodeColor=<rgba 0 0.2031 1 1>}
  NamedRerouteDeclaration(None=ComponentMask(None=@n71) {r=true}) {name=Cloth Mask, nodeColor=<rgba 0 0.625 1 1>}
  n73 = Multiply(
   A=Power(
    Base=MaterialFunctionCall!(InAmplitude, InSpectrum, CustomFakeSpectrum, CustomFakeAmplitude) {materialFunction=/BRCosmetics/Global/MaterialFunctions/AudioReactivity/MF_AudioReactivity_FrontendLoop}.Amplitude
    Exp=ScalarParameter() {defaultValue=6, group=CanineCronutDig_Ware, parameterName=Amplitude Power}
   )
   B=ScalarParameter() {defaultValue=1, group=CanineCronutDig_Ware, parameterName=Amplitude Boost}
  )
  NamedRerouteDeclaration(None=LinearInterpolate(B=Saturate(None=@n73), Alpha=ScalarParameter() {defaultValue=1, group=CanineCronutDig_Ware, parameterName=Is Reactive?})) {name=Audio Reactivity, nodeColor=<rgba 0 0.2969 1 1>}
  n74 = LinearInterpolate(
   A=ScalarParameter() {defaultValue=4, group=CanineCronutDig_Ware, parameterName=Fresnel Colour - Exp}
   B=ScalarParameter() {defaultValue=3, parameterName=Fresnel Colour - Exp Reactive}
   Alpha=@n3
  ) {constB=2}
  n75 = VertexNormalWS()
  n76 = LinearInterpolate(A=PixelNormalWS(), B=@n75, Alpha=ScalarParameter() {defaultValue=0.5, group=CanineCronutDig_Ware, parameterName=Flatten Fresnel Normal})
  n77 = VectorParameter() {group=CanineCronutDig_Ware, parameterName=Reactive Color, defaultValue=<rgba 0.0052 0.2232 0.0578 1>}
  n78 = VectorParameter() {parameterName=Reactive Color 2, defaultValue=<rgba 0.0082 0.2195 0.1166 1>}
  n79 = NamedRerouteUsage()
  n80 = Multiply(
   A=Fresnel(ExponentIn=ScalarParameter() {defaultValue=1.57, group=CanineCronutDig_Ware, parameterName=Max Fresnel Color - Exp}, Normal=@n75) {baseReflectFraction=0}
   B=ScalarParameter() {defaultValue=20, group=CanineCronutDig_Ware, parameterName=Max Fresnel Colour - Brightness}
  )
  n81 = ScalarParameter() {defaultValue=10, group=CanineCronutDig_Ware, parameterName=Fresnel Colour - Max}
  n82 = Multiply(
   A=Multiply(A=Fresnel(ExponentIn=@n74, Normal=@n76) {baseReflectFraction=0}, B=LinearInterpolate(A=@n77.RGB, B=@n78.RGB, Alpha=@n79))
   B=LinearInterpolate(A=ScalarParameter() {group=CanineCronutDig_Ware, parameterName=Fresnel Colour - Min}, B=Max(A=@n80, B=@n81), Alpha=NamedRerouteUsage())
  )
  n83 = NamedRerouteUsage()
  NamedRerouteDeclaration(None=LinearInterpolate(A=@n82, B=Multiply(A=@n82, B=@n83), Alpha=ScalarParameter() {group=CanineCronutDig_Ware, parameterName=Reactive Mask Override})) {name=Fresnel Colour, nodeColor=<rgba 0 1 0.2734 1>}
  NamedRerouteUsage()
  Saturate(None=@n47)
  n84 = GetMaterialAttributes(None=NamedRerouteUsage()) {attributeGetTypes [#4]}
  n85 = VertexNormalWS()
  n86 = LinearInterpolate(
   A=Multiply(A=@n85, B=ScalarParameter() {defaultValue=0.344, parameterName=Param_4})
   B=Multiply(A=@n85, B=ScalarParameter() {defaultValue=2, parameterName=Param_3})
   Alpha=NamedRerouteUsage()
  )
  n87 = NamedRerouteUsage()
  n88 = Multiply(A=@n5, B=Fresnel(ExponentIn=ScalarParameter() {defaultValue=2, parameterName=Param_5}) {baseReflectFraction=0})
  n89 = NamedRerouteUsage()
  SetMaterialAttributes!(
   MaterialAttributes=BlendMaterialAttributes(
    A=NamedRerouteUsage()
    B=SetMaterialAttributes!(MaterialAttributes=@n84.MaterialAttributes, WorldPositionOffset=@n86, EmissiveColor=@n87, ShadingModel) {attributeSetTypes [#7, #1, #6]}
    Alpha=@n5
   )
   WorldPositionOffset
   Opacity=Multiply(A=Multiply(A=@n88, B=@n89), B=ScalarParameter() {defaultValue=1, parameterName=Param_7})
   ShadingModel
  ) {attributeSetTypes [#7, #8, #6]}
  Subtract(A=NamedRerouteUsage(), B=NamedRerouteUsage())
  Subtract(A=@n4, B=@n3)
  n90 = Divide(A=ComponentMask(None=TextureCoordinate()) {g=true}) {constB=64}
  n91 = ScalarParameter() {group=Audio Reactivity, parameterName=FakeAudioCurveIndex}
  TextureSample(UVs=AppendVector(A=@n67, B=Add(A=@n90, B=@n91))) {samplerType=SAMPLERTYPE_LinearColor, texture=&CA_Frontend_FakeAudioReactivity}
 ]}
