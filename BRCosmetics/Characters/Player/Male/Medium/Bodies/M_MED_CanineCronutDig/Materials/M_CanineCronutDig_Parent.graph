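# Parent material for the CanineCronutDig male character body, built on material attributes with per-expression shading models and usage flags for skeletal, cloth, morph and Nanite rendering.
# The shared 2023 character PreFX function feeds the skin-specific Ware function, whose output passes into the shared PostFX function.
# A "Debug TA Masks" static switch can override the final Base Color through a named reroute for mask visualization.
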Material /BRCosmetics/Characters/Player/Male/Medium/Bodies/M_MED_CanineCronutDig/Materials/M_CanineCronutDig_Parent
graph {
 properties {
  bUseMaterialAttributes=true
  bUsedWithClothing=true
  bUsedWithMeshDeformer=true
  bUsedWithMorphTargets=true
  bUsedWithNanite=true
  bUsedWithNiagaraMeshParticles=true
  bUsedWithSkeletalMesh=true
  pixelDepthOffsetMode=PDOM_Legacy
  shadingModel=MSM_FromMaterialExpression
  usedShadingModels=Default Lit | Subsurface | Clear Coat | Subsurface Profile | Cloth
 }
 nodes [
  n1 = MaterialFunctionCall() {materialFunction=&MF_CharacterParent_2023_PreFX}
  n2 = MaterialFunctionCall(
   Material=MaterialFunctionCall!(
    Result=@n1.Result
    M=@n1.Result
    Fresnel=@n1.Result
    Tech Art Mask=NamedRerouteDeclaration(None=@n1.Tech Art Mask) {name=TA Mask, nodeColor=<rgba 0.3828 1 0 1>}
   ) {materialFunction=&/MF_CanineCronutDig_Ware_Parent}.MA
   M=@n1.M
   Pre Skinned Local Position=@n1.M
   Distance Blend=@n1.M
   Fresnel=@n1.M
  ) {materialFunction=&MF_CharacterParent_2023_PostFX}
  StaticSwitchParameter(
   True=SetMaterialAttributes!(MaterialAttributes=@n2.Result, Base Color=NamedRerouteUsage()) {attributeSetTypes [#1]}
   False=@n2.Result
  ) {parameterName=Debug TA Masks}
 ]
}
