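# Control Rig for the CanineCronutDig medium male player body, adding corrective deformation to its outfit bones.
# It runs a shared DeformRig function, then sequences correctives for braided ropes, cords, clavicle buttons, jacket hips and sleeves.
# Euler angles from clavicles, hands and spine, plus spherical pose readers on upper arms and thighs, weight additive transform offsets.
# Parent, position and scale constraints attach buttons, rope ends and the jacket pelvic deformer.
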ControlRigBlueprint /BRCosmetics/Characters/Player/Male/Medium/Bodies/M_MED_CanineCronutDig/Meshes/M_MED_CanineCronutDig_CtrlRig
graphs [
 RigVMFunctionLibrary {links [], nodes []}
 RigVMFunction_Sequence_2_ContainedGraph {nodes [
   n1 = RigVMFunctionEntryNode()
   n2 = &RigVMFunction_Sequence(ExecuteContext=@n1.ExecuteContext)
   n3 = &RigVMFunction_Sequence(ExecuteContext=@n2.B)
   n4 = &RigVMFunction_Sequence(ExecuteContext=@n3.B)
   n5 = &RigVMFunction_Sequence(ExecuteContext=@n4.B)
   RigVMFunctionReturnNode(A=@n2.A, B=@n3.A, C=@n4.A, D=@n5.A, E=@n5.B)
  ]}
 RigVMFunction_Sequence_3_ContainedGraph {nodes [
   n1 = RigVMFunctionEntryNode()
   n2 = &RigVMFunction_Sequence(ExecuteContext=@n1.ExecuteContext)
   n3 = &RigVMFunction_Sequence(ExecuteContext=@n2.B)
   n4 = &RigVMFunction_Sequence(ExecuteContext=@n3.B)
   RigVMFunctionReturnNode(B=@n3.A, A=@n2.A, C=@n4.A, D=@n4.B)
  ]}
 RigVMModel {nodes [
   RigUnit_BeginExecution = &RigUnit_BeginExecution()
   FN_DeformRig_FNC = RigVMFunctionReferenceNode(
    DeformRig_ Curve="(deformRig_l_upperarm_up_alpha_72_#1=False,deformRig_r_upperarm_up_alpha_73_#2=False,deformRig_l_upperarm_fr_alpha_74_#3=False,deformRig_r_upperarm_fr_alpha_75_CB51B8 ...(truncated)"
    DeformRig_Bool="(upperarm_1_#4=True,elbow_3_#5=True,wrist_7_#6=True,groin_8_#7=True,glute_17_#8=True,knee_9_40C408324CA261 ...(truncated)"
    DeformRig_Debug=(upperarm_debug_18_#4=False,wrist_debug_20_#6=False,lat_debug_26_#9=False,groinGlute_debug_27_#7=False)
    DeformRig_Value="(upperarm_up_18_#10=(Rotation=(X=0,Y=0,Z=0,W=1),Translation=(X=0,Y=0,Z=8),Scale3D=(X=0.7,Y=1,Z=1)),upperarm_fwd_fr_19_#11=(Rotation=(X=0,Y=0,Z=0,W=1),Translation=(X=-10,Y=10,Z=0),Scale3D=(X=0.5,Y=1, ...(truncated)"
    ExecuteContext=@RigUnit_BeginExecution.ExecutePin
   )
   > RigVMFunction_Sequence_1 = RigVMAggregateNode()
   &RigUnit_ParentConstraint(
    AdvancedSettings=(InterpolationType=Average,RotationOrderForFilter=XZY)
    Child=(Type=Bone,Name="braidedRopeBack5")
    Parents=((Item=(Type=Bone,Name="CordsBot"),Weight=1))
    Weight=1
    ExecutePin=@RigVMFunction_Sequence_1.B
   )
   > &RigUnit_ParentConstraint(
    AdvancedSettings=(InterpolationType=Average,RotationOrderForFilter=XZY)
    Child=(Type=Bone,Name="braidedRopeBack6")
    Parents=((Item=(Type=Bone,Name="CordsBot"),Weight=1))
    Weight=1
   )
   > &RigUnit_ParentConstraint(
    AdvancedSettings=(InterpolationType=Average,RotationOrderForFilter=XZY)
    Child=(Type=Bone,Name="braidedRopeFront7")
    Parents=((Item=(Type=Bone,Name="CordsBot"),Weight=1))
    Weight=1
   )
   > &RigUnit_ParentConstraint(
    AdvancedSettings=(InterpolationType=Average,RotationOrderForFilter=XZY)
    Child=(Type=Bone,Name="braidedRopeFront6")
    Parents=((Item=(Type=Bone,Name="CordsBot"),Weight=1))
    Weight=1
   )
   > &RigUnit_ParentConstraint(
    AdvancedSettings=(InterpolationType=Average,RotationOrderForFilter=XZY)
    Child=(Type=Bone,Name="button_clavicle_r")
    Parents="((Item=(Type=Bone,Name=\"neck_01\"),Weight=0.0174),(Item=(Type=Bone,Name=\"neck_02\"),Weight=0.0502),(Item=(Type=Bone,Name=\"head\"),Weight=0.0012),(Item=(Type=Bone,Name=\"deform_upperarm_bk_r\"),Weight=0.004),(Item=(Type=Bone,Name=\"deform_upperarm_up_r\"),Weight=0 ...(truncated)"
    Weight=1
   )
   > &RigUnit_ParentConstraint(
    AdvancedSettings=(InterpolationType=Average,RotationOrderForFilter=XZY)
    Child=(Type=Bone,Name="button_clavicle_l")
    Parents="((Item=(Type=Bone,Name=\"neck_01\"),Weight=0.0174),(Item=(Type=Bone,Name=\"neck_02\"),Weight=0.0502),(Item=(Type=Bone,Name=\"head\"),Weight=0.0012),(Item=(Type=Bone,Name=\"deform_upperarm_bk_l\"),Weight=0.004),(Item=(Type=Bone,Name=\"deform_upperarm_up_l\"),Weight=0 ...(truncated)"
    Weight=1
   )
   n1 = &RigVMFunction_MathQuaternionToEuler(
    RotationOrder=ZYX
    Value=&RigUnit_GetRelativeTransformForItem(Child=(Type=Bone,Name="clavicle_l"), Parent=(Type=Null,Name="clavicle_l_null"), bChildInitial=False, bParentInitial=False).RelativeTransform.Rotation
   )
   > &RigUnit_ModifyTransforms(
    ItemToModify=((Item=(Type=Bone,Name="button_clavicle_l"),Transform=(Rotation=(X=0,Y=-0.3231,Z=0,W=0.9464),Translation=(X=0.1771,Y=0,Z=0.5698),Scale3D=(X=1,Y=1,Z=1))))
    WeightMinimum=0
    Weight=&RigVMFunction_MathDoubleRemap(SourceMaximum=-30, SourceMinimum=0, TargetMaximum=1, TargetMinimum=0, Value=@n1.Result.Y).Result
   )
   > &RigUnit_ModifyTransforms(
    ItemToModify=((Item=(Type=Bone,Name="button_clavicle_l"),Transform=(Rotation=(X=0,Y=0,Z=0,W=1),Translation=(X=0,Y=-1.3242,Z=0),Scale3D=(X=1,Y=1,Z=1))))
    WeightMinimum=0
    Weight=&RigVMFunction_MathDoubleRemap(SourceMaximum=32, SourceMinimum=0, TargetMaximum=1, TargetMinimum=0, Value=@n1.Result.X).Result
   )
   n2 = &RigVMFunction_MathQuaternionToEuler(
    RotationOrder=ZYX
    Value=&RigUnit_GetRelativeTransformForItem(Child=(Type=Bone,Name="clavicle_r"), Parent=(Type=Null,Name="clavicle_r_null"), bChildInitial=False, bParentInitial=False).RelativeTransform.Rotation
   )
   > &RigUnit_ModifyTransforms(
    ItemToModify=((Item=(Type=Bone,Name="button_clavicle_r"),Transform=(Rotation=(X=0,Y=-0.3231,Z=0,W=0.9464),Translation=(X=-0.1771,Y=0,Z=-0.5698),Scale3D=(X=1,Y=1,Z=1))))
    WeightMinimum=0
    Weight=&RigVMFunction_MathDoubleRemap(SourceMaximum=-30, SourceMinimum=0, TargetMaximum=1, TargetMinimum=0, Value=@n2.Result.Y).Result
   )
   > &RigUnit_ModifyTransforms(
    ItemToModify=((Item=(Type=Bone,Name="button_clavicle_r"),Transform=(Rotation=(X=0,Y=0,Z=0,W=1),Translation=(X=0,Y=1.3242,Z=0),Scale3D=(X=1,Y=1,Z=1))))
    WeightMinimum=0
    Weight=&RigVMFunction_MathDoubleRemap(SourceMaximum=-32, SourceMinimum=0, TargetMaximum=1, TargetMinimum=0, Value=@n2.Result.X).Result
   )
   n3 = &RigUnit_SphericalPoseReader(
    ActiveRegionScaleFactors=(PositiveWidth=1,NegativeWidth=2,PositiveHeight=2,NegativeHeight=1)
    ActiveRegionSize=0.1
    Debug=(bDrawDebug=false,bDraw2D=False,bDrawLocalAxes=False,DebugScale=25,DebugSegments=20,DebugThickness=0.25)
    DriverAxis=(X=1,Y=0,Z=0)
    DriverItem=(Type=Bone,Name="thigh_r")
    FalloffRegionScaleFactors=(PositiveWidth=1,NegativeWidth=1,PositiveHeight=1,NegativeHeight=1)
    FalloffSize=0.35
    FlipHeightScaling=False
    FlipWidthScaling=False
    OptionalParentItem=(Type=Bone,Name="None")
    RotationOffset=(X=-30,Y=30,Z=0)
    ExecutePin=@RigVMFunction_Sequence_1.C
   )
   > n4 = &RigUnit_SphericalPoseReader(
    ActiveRegionScaleFactors=(PositiveWidth=1,NegativeWidth=2,PositiveHeight=2,NegativeHeight=1)
    ActiveRegionSize=0.1
    Debug=(bDrawDebug=false,bDraw2D=False,bDrawLocalAxes=False,DebugScale=25,DebugSegments=20,DebugThickness=0.25)
    DriverAxis=(X=1,Y=0,Z=0)
    DriverItem=(Type=Bone,Name="thigh_l")
    FalloffRegionScaleFactors=(PositiveWidth=1,NegativeWidth=1,PositiveHeight=1,NegativeHeight=1)
    FalloffSize=0.35
    FlipHeightScaling=False
    FlipWidthScaling=False
    OptionalParentItem=(Type=Bone,Name="None")
    RotationOffset=(X=-30,Y=30,Z=0)
   )
   > &RigUnit_ModifyTransforms(
    ItemToModify=((Item=(Type=Bone,Name="deform_hip_r"),Transform=(Rotation=(X=-0.2818,Y=0.3161,Z=0.0831,W=0.9021),Translation=(X=10.6548,Y=4.7819,Z=-13.4956),Scale3D=(X=1,Y=1,Z=1))))
    WeightMinimum=0
    Weight=@n3.OutputParam
   )
   > &RigUnit_ModifyTransforms(
    ItemToModify=((Item=(Type=Bone,Name="deform_hip_l"),Transform=(Rotation=(X=-0.2818,Y=0.3161,Z=0.0831,W=0.9021),Translation=(X=-10.6548,Y=-4.7819,Z=13.4956),Scale3D=(X=1,Y=1,Z=1))))
    WeightMinimum=0
    Weight=@n4.OutputParam
   )
   > /Script/ControlRig.RigUnit_PositionConstraintLocalSpaceOffset(
    Child=(Type=Bone,Name="Jacket_pelvic_Deformer")
    Filter=(bX=True,bY=True,bZ=True)
    Parents=((Item=(Type=Bone,Name="deform_hip_r"),Weight=1),(Item=(Type=Bone,Name="deform_hip_l"),Weight=1))
    Weight=1
    bMaintainOffset=True
   ) {is=struct RigVMUnitNode}
   RigVMFunction_Sequence = &RigVMFunction_Sequence(ExecuteContext=@RigVMFunction_Sequence_1.D)
   n5 = &RigVMFunction_MathQuaternionToEuler(
    RotationOrder=ZYX
    Value=&RigUnit_GetRelativeTransformForItem(Child=(Type=Bone,Name="hand_r"), Parent=(Type=Null,Name="hand_r_null"), bChildInitial=False, bParentInitial=False).RelativeTransform.Rotation
   )
   &RigUnit_ModifyTransforms(
    ItemToModify=((Item=(Type=Bone,Name="None"),Transform=(Translation=(X=0,Y=0,Z=0),Scale3D=(X=1,Y=1,Z=1))))
    WeightMinimum=0
    Weight=&RigVMFunction_MathDoubleRemap(SourceMaximum=-60, SourceMinimum=0, TargetMaximum=1, TargetMinimum=0, bClamp=False, Value=@n5.Result.Z).Result
    ExecutePin=@RigVMFunction_Sequence.A
   )
   > &RigUnit_ModifyTransforms(
    ItemToModify="((Item=(Type=Bone,Name=\"sleeve_r\"),Transform=(Rotation=(X=0,Y=0,Z=0.3927,W=-0.9197),Translation=(X=-0.96,Y=-1.7456,Z=0),Scale3D=(X=1,Y=1.5084,Z=1))),(Item=(Type=Bone,Name=\"thumb_Sleeve_r\"),Transform=(Rotation=(X=0,Y=-0.4592,Z=0,W=-0.8883),Translation=(X=1. ...(truncated)"
    WeightMinimum=0
    Weight=&RigVMFunction_MathDoubleRemap(SourceMaximum=90, SourceMinimum=0, TargetMaximum=1, TargetMinimum=0, bClamp=False, Value=@n5.Result.Z).Result
   )
   > &RigUnit_ModifyTransforms(
    ItemToModify="((Item=(Type=Bone,Name=\"sleeve_r\"),Transform=(Rotation=(X=0.2831,Y=0,Z=0,W=-0.9591),Translation=(X=0,Y=0.4506,Z=0.6968),Scale3D=(X=1,Y=1.3515,Z=0.9862))),(Item=(Type=Bone,Name=\"thumb_Sleeve_r\"),Transform=(Rotation=(X=-0.1128,Y=0.0642,Z=-0.3385,W=0.932),Tra ...(truncated)"
    WeightMaximum=0
    WeightMinimum=0
    Weight=&RigVMFunction_MathDoubleRemap(SourceMaximum=80, SourceMinimum=0, TargetMaximum=1, TargetMinimum=0, bClamp=False, Value=@n5.Result.X).Result
   )
   > &RigUnit_ModifyTransforms(
    ItemToModify="((Item=(Type=Bone,Name=\"sleeve_r\"),Transform=(Rotation=(X=-0.4708,Y=0,Z=0,W=-0.8823),Translation=(X=0,Y=0.2646,Z=-0.8959),Scale3D=(X=1,Y=1.5148,Z=1))),(Item=(Type=Bone,Name=\"thumb_Sleeve_r\"),Transform=(Rotation=(X=-0.2677,Y=0.0696,Z=-0.2417,W=-0.9301),Tran ...(truncated)"
    WeightMaximum=0
    WeightMinimum=0
    Weight=&RigVMFunction_MathDoubleRemap(SourceMaximum=-80, SourceMinimum=0, TargetMaximum=1, TargetMinimum=0, bClamp=False, Value=@n5.Result.X).Result
   )
   > &RigUnit_ModifyTransforms(
    ItemToModify="((Item=(Type=Bone,Name=\"sleeve_r\"),Transform=(Rotation=(X=0,Y=-0.204,Z=0,W=-0.979),Translation=(X=0.2456,Y=0,Z=-1.2906),Scale3D=(X=1,Y=1,Z=1.3299))),(Item=(Type=Bone,Name=\"thumb_Sleeve_r\"),Transform=(Rotation=(X=0.0552,Y=0.1263,Z=0.3518,W=0.9259),Translati ...(truncated)"
    WeightMinimum=0
    Weight=&RigVMFunction_MathDoubleRemap(SourceMaximum=-70, SourceMinimum=0, TargetMaximum=1, TargetMinimum=0, bClamp=False, Value=@n5.Result.Y).Result
   )
   > &RigUnit_ModifyTransforms(
    ItemToModify="((Item=(Type=Bone,Name=\"sleeve_r\"),Transform=(Rotation=(X=0,Y=0.17,Z=0,W=-0.9854),Translation=(X=0.3685,Y=0,Z=0.5854),Scale3D=(X=1,Y=1,Z=1.1273))),(Item=(Type=Bone,Name=\"thumb_Sleeve_r\"),Transform=(Rotation=(X=0.174,Y=-0.0033,Z=-0.1381,W=0.975),Translation ...(truncated)"
    WeightMinimum=0
    Weight=&RigVMFunction_MathDoubleRemap(SourceMaximum=40, SourceMinimum=0, TargetMaximum=1, TargetMinimum=0, bClamp=False, Value=@n5.Result.Y).Result
   )
   > &RigUnit_ParentConstraint(
    AdvancedSettings=(InterpolationType=Average,RotationOrderForFilter=XZY)
    Child=(Type=Bone,Name="button_hand_r")
    Parents="((Item=(Type=Bone,Name=\"sleeve_r\"),Weight=0.4686),(Item=(Type=Bone,Name=\"lowerarm_r\"),Weight=0.0166),(Item=(Type=Bone,Name=\"deform_wrist_dn_r\"),Weight=0.0053),(Item=(Type=Bone,Name=\"deform_wrist_up_r\"),Weight=0.0002),(Item=(Type=Bone,Name=\"lowerarm_twist_0 ...(truncated)"
    Weight=1
   )
   > /Script/ControlRig.RigUnit_ScaleConstraintLocalSpaceOffset(
    Child=(Type=Bone,Name="button_hand_r")
    Filter=(bX=True,bY=True,bZ=True)
    Parents=((Item=(Type=Bone,Name="lowerarm_r"),Weight=1))
    Weight=1
    bMaintainOffset=True
   ) {is=struct RigVMUnitNode}
   n6 = &RigVMFunction_MathQuaternionToEuler(
    RotationOrder=ZYX
    Value=&RigUnit_GetRelativeTransformForItem(Child=(Type=Bone,Name="hand_l"), Parent=(Type=Null,Name="hand_l_null"), bChildInitial=False, bParentInitial=False).RelativeTransform.Rotation
   )
   &RigUnit_ModifyTransforms(
    ItemToModify=((Item=(Type=Bone,Name="None"),Transform=(Translation=(X=0,Y=0,Z=0),Scale3D=(X=1,Y=1,Z=1))))
    WeightMinimum=0
    Weight=&RigVMFunction_MathDoubleRemap(SourceMaximum=-60, SourceMinimum=0, TargetMaximum=1, TargetMinimum=0, bClamp=False, Value=@n6.Result.Z).Result
    ExecutePin=@RigVMFunction_Sequence.B
   )
   > &RigUnit_ModifyTransforms(
    ItemToModify="((Item=(Type=Bone,Name=\"sleeve_l\"),Transform=(Rotation=(X=0,Y=0,Z=0.3927,W=-0.9197),Translation=(X=0.96,Y=1.7456,Z=0),Scale3D=(X=1,Y=1.5084,Z=1))),(Item=(Type=Bone,Name=\"thumb_Sleeve_l\"),Transform=(Rotation=(X=0,Y=-0.4592,Z=0,W=-0.8883),Translation=(X=-1.0 ...(truncated)"
    WeightMinimum=0
    Weight=&RigVMFunction_MathDoubleRemap(SourceMaximum=90, SourceMinimum=0, TargetMaximum=1, TargetMinimum=0, bClamp=False, Value=@n6.Result.Z).Result
   )
   > &RigUnit_ModifyTransforms(
    ItemToModify="((Item=(Type=Bone,Name=\"sleeve_l\"),Transform=(Rotation=(X=0.2831,Y=0,Z=0,W=-0.9591),Translation=(X=0,Y=-0.4506,Z=-0.6968),Scale3D=(X=1,Y=1.3515,Z=0.9862))),(Item=(Type=Bone,Name=\"thumb_Sleeve_l\"),Transform=(Rotation=(X=-0.1128,Y=0.0642,Z=-0.3385,W=0.932),T ...(truncated)"
    WeightMaximum=0
    WeightMinimum=0
    Weight=&RigVMFunction_MathDoubleRemap(SourceMaximum=80, SourceMinimum=0, TargetMaximum=1, TargetMinimum=0, bClamp=False, Value=@n6.Result.X).Result
   )
   > &RigUnit_ModifyTransforms(
    ItemToModify="((Item=(Type=Bone,Name=\"sleeve_l\"),Transform=(Rotation=(X=-0.4708,Y=0,Z=0,W=-0.8823),Translation=(X=0,Y=-0.2646,Z=0.8959),Scale3D=(X=1,Y=1.5148,Z=1))),(Item=(Type=Bone,Name=\"thumb_Sleeve_l\"),Transform=(Rotation=(X=-0.2677,Y=0.0696,Z=-0.2417,W=-0.9301),Tran ...(truncated)"
    WeightMaximum=0
    WeightMinimum=0
    Weight=&RigVMFunction_MathDoubleRemap(SourceMaximum=-80, SourceMinimum=0, TargetMaximum=1, TargetMinimum=0, bClamp=False, Value=@n6.Result.X).Result
   )
   > &RigUnit_ModifyTransforms(
    ItemToModify=((Item=(Type=Bone,Name="sleeve_l"),Transform=(Rotation=(X=0,Y=0,Z=0,W=1),Translation=(X=0,Y=0,Z=0),Scale3D=(X=1,Y=1,Z=1))),(Item=(Type=Bone,Name="None"),Transform=(Rotation=(X=0,Y=0,Z=0,W=1),Translation=(X=0,Y=0,Z=0),Scale3D=(X=1,Y=1,Z=1))))
    WeightMinimum=0
    Weight=&RigVMFunction_MathDoubleRemap(SourceMaximum=-70, SourceMinimum=0, TargetMaximum=1, TargetMinimum=0, bClamp=False, Value=@n6.Result.Y).Result
   )
   > &RigUnit_ModifyTransforms(
    ItemToModify="((Item=(Type=Bone,Name=\"sleeve_l\"),Transform=(Rotation=(X=0,Y=0.17,Z=0,W=-0.9854),Translation=(X=-0.3685,Y=0,Z=-0.5854),Scale3D=(X=1,Y=1,Z=1.1273))),(Item=(Type=Bone,Name=\"thumb_Sleeve_l\"),Transform=(Rotation=(X=0.174,Y=-0.0033,Z=-0.1381,W=0.975),Translati ...(truncated)"
    WeightMinimum=0
    Weight=&RigVMFunction_MathDoubleRemap(SourceMaximum=40, SourceMinimum=0, TargetMaximum=1, TargetMinimum=0, bClamp=False, Value=@n6.Result.Y).Result
   )
   > &RigUnit_ParentConstraint(
    AdvancedSettings=(InterpolationType=Average,RotationOrderForFilter=XZY)
    Child=(Type=Bone,Name="button_hand_l")
    Parents="((Item=(Type=Bone,Name=\"sleeve_l\"),Weight=0.4686),(Item=(Type=Bone,Name=\"lowerarm_l\"),Weight=0.0166),(Item=(Type=Bone,Name=\"deform_wrist_dn_l\"),Weight=0.0053),(Item=(Type=Bone,Name=\"deform_wrist_up_l\"),Weight=0.0002),(Item=(Type=Bone,Name=\"lowerarm_twist_0 ...(truncated)"
    Weight=1
   )
   > &RigUnit_ScaleConstraintLocalSpaceOffset(
    Child=(Type=Bone,Name="button_hand_l")
    Filter=(bX=True,bY=True,bZ=True)
    Parents=((Item=(Type=Bone,Name="lowerarm_l"),Weight=1))
    Weight=1
    bMaintainOffset=True
   )
   RigVMFunction_Sequence_2 = RigVMAggregateNode(ExecuteContext=@RigVMFunction_Sequence_1.A)
   n7 = &RigUnit_SphericalPoseReader(
    ActiveRegionScaleFactors=(PositiveWidth=1,NegativeWidth=1,PositiveHeight=1,NegativeHeight=1)
    ActiveRegionSize=0.1
    Debug=(bDrawDebug=true,bDraw2D=False,bDrawLocalAxes=False,DebugScale=25,DebugSegments=20,DebugThickness=0.25)
    DriverAxis=(X=-1,Y=0,Z=0)
    DriverItem=(Type=Bone,Name="upperarm_r")
    FalloffRegionScaleFactors=(PositiveWidth=1,NegativeWidth=1,PositiveHeight=1,NegativeHeight=1)
    FalloffSize=0.4
    FlipHeightScaling=False
    FlipWidthScaling=False
    OptionalParentItem=(Type=Bone,Name="None")
    RotationOffset=(X=190,Y=55,Z=-30)
    ExecutePin=@RigVMFunction_Sequence_2.D
   )
   > n8 = &RigUnit_SphericalPoseReader(
    ActiveRegionScaleFactors=(PositiveWidth=1,NegativeWidth=1,PositiveHeight=1,NegativeHeight=1)
    ActiveRegionSize=0.1
    Debug=(bDrawDebug=false,bDraw2D=False,bDrawLocalAxes=False,DebugScale=25,DebugSegments=20,DebugThickness=0.25)
    DriverAxis=(X=-1,Y=0,Z=0)
    DriverItem=(Type=Bone,Name="upperarm_l")
    FalloffRegionScaleFactors=(PositiveWidth=1,NegativeWidth=1,PositiveHeight=1,NegativeHeight=1)
    FalloffSize=0.3
    FlipHeightScaling=False
    FlipWidthScaling=False
    OptionalParentItem=(Type=Bone,Name="None")
    RotationOffset=(X=190,Y=55,Z=-30)
   )
   > &RigUnit_ModifyTransforms(
    ItemToModify="((Item=(Type=Bone,Name=\"VollumeCompensator_r\"),Transform=(Rotation=(X=0.0119,Y=0.0004,Z=-0.037,W=-0.9992),Translation=(X=0.0001,Y=-1.8361,Z=0.4386),Scale3D=(X=1,Y=1.35,Z=1.2))),(Item=(Type=Bone,Name=\"cord_C\"),Transform=(Rotation=(X=-0.0363,Y=-0.0971,Z=-0.4 ...(truncated)"
    WeightMinimum=0
    Weight=@n7.OutputParam
   )
   > &RigUnit_ModifyTransforms(
    ItemToModify=((Item=(Type=Bone,Name="VollumeCompensator_l"),Transform=(Rotation=(X=0,Y=0,Z=-0.037,W=-0.9993),Translation=(X=0,Y=1.8363,Z=-0.4387),Scale3D=(X=1,Y=1.35,Z=1.2))))
    WeightMinimum=0
    Weight=@n8.OutputParam
   )
   > n9 = &RigUnit_SphericalPoseReader(
    ActiveRegionScaleFactors=(PositiveWidth=1,NegativeWidth=1,PositiveHeight=1,NegativeHeight=1)
    ActiveRegionSize=0.1
    Debug=(bDrawDebug=true,bDraw2D=False,bDrawLocalAxes=False,DebugScale=25,DebugSegments=20,DebugThickness=0.25)
    DriverAxis=(X=-1,Y=0,Z=0)
    DriverItem=(Type=Bone,Name="upperarm_r")
    FalloffRegionScaleFactors=(PositiveWidth=1,NegativeWidth=1,PositiveHeight=1,NegativeHeight=1)
    FalloffSize=0.2
    FlipHeightScaling=False
    FlipWidthScaling=False
    OptionalParentItem=(Type=Bone,Name="clavicle_r")
    RotationOffset=(X=23.992,Y=45.3924,Z=0)
   )
   > &RigUnit_ModifyTransforms(
    ItemToModify="((Item=(Type=Bone,Name=\"braidedRopeBack2\"),Transform=(Rotation=(X=0,Y=0,Z=0,W=-1),Translation=(X=0,Y=0,Z=1.0448),Scale3D=(X=1,Y=1,Z=1))),(Item=(Type=Bone,Name=\"braidedRopeBack3\"),Transform=(Rotation=(X=0,Y=0,Z=0,W=-1),Translation=(X=0,Y=-4.0411,Z=0),Scale3 ...(truncated)"
    WeightMinimum=0
    Weight=@n9.OutputParam
   )
   > n10 = &RigUnit_SphericalPoseReader(
    ActiveRegionScaleFactors=(PositiveWidth=1,NegativeWidth=1,PositiveHeight=1,NegativeHeight=1)
    ActiveRegionSize=0.08
    Debug=(bDrawDebug=true,bDraw2D=False,bDrawLocalAxes=False,DebugScale=25,DebugSegments=20,DebugThickness=0.25)
    DriverAxis=(X=-1,Y=0,Z=0)
    DriverItem=(Type=Bone,Name="upperarm_r")
    FalloffRegionScaleFactors=(PositiveWidth=1,NegativeWidth=1,PositiveHeight=1,NegativeHeight=1)
    FalloffSize=0.3
    FlipHeightScaling=False
    FlipWidthScaling=False
    OptionalParentItem=(Type=Bone,Name="None")
    RotationOffset=(X=194.241,Y=97.451,Z=-48.7221)
   )
   > &RigUnit_ModifyTransforms(
    ItemToModify="((Item=(Type=Bone,Name=\"braidedRopeFront2\"),Transform=(Rotation=(X=-0.015,Y=0.0601,Z=0.241,W=0.9685),Translation=(X=-0.9947,Y=1.875,Z=0),Scale3D=(X=1,Y=1,Z=1))),(Item=(Type=Bone,Name=\"braidedRopeFront3\"),Transform=(Rotation=(X=-0.0001,Y=-0.0008,Z=0.1628,W= ...(truncated)"
    WeightMinimum=0
    Weight=@n10.OutputParam
   )
   > n11 = &RigUnit_SphericalPoseReader(
    ActiveRegionScaleFactors=(PositiveWidth=1,NegativeWidth=1.5,PositiveHeight=1,NegativeHeight=1)
    ActiveRegionSize=0.03
    Debug=(bDrawDebug=true,bDraw2D=False,bDrawLocalAxes=False,DebugScale=25,DebugSegments=20,DebugThickness=0.25)
    DriverAxis=(X=-1,Y=0,Z=0)
    DriverItem=(Type=Bone,Name="upperarm_r")
    FalloffRegionScaleFactors=(PositiveWidth=1,NegativeWidth=1,PositiveHeight=1,NegativeHeight=1)
    FalloffSize=0.25
    FlipHeightScaling=False
    FlipWidthScaling=False
    OptionalParentItem=(Type=Bone,Name="clavicle_r")
    RotationOffset=(X=-34.5829,Y=19.6729,Z=11.4026)
   )
   > &RigUnit_ModifyTransforms(
    ItemToModify="((Item=(Type=Bone,Name=\"braidedRopeFront2\"),Transform=(Rotation=(X=0.0064,Y=-0.0363,Z=-0.1746,W=-0.9839),Translation=(X=-0.2172,Y=0.5925,Z=0),Scale3D=(X=1,Y=1,Z=1))),(Item=(Type=Bone,Name=\"braidedRopeFront3\"),Transform=(Rotation=(X=0,Y=0,Z=0,W=-1),Translat ...(truncated)"
    WeightMinimum=0
    Weight=@n11.OutputParam
   )
   > n12 = &RigUnit_SphericalPoseReader(
    ActiveRegionScaleFactors=(PositiveWidth=1,NegativeWidth=1,PositiveHeight=1,NegativeHeight=1)
    ActiveRegionSize=0.03
    Debug=(bDrawDebug=true,bDraw2D=False,bDrawLocalAxes=False,DebugScale=25,DebugSegments=20,DebugThickness=0.25)
    DriverAxis=(X=-1,Y=0,Z=0)
    DriverItem=(Type=Bone,Name="upperarm_r")
    FalloffRegionScaleFactors=(PositiveWidth=1,NegativeWidth=1,PositiveHeight=1,NegativeHeight=1)
    FalloffSize=0.25
    FlipHeightScaling=False
    FlipWidthScaling=False
    OptionalParentItem=(Type=Bone,Name="clavicle_r")
    RotationOffset=(X=-55.8604,Y=-4.7644,Z=6.7182)
   )
   > &RigUnit_ModifyTransforms(
    ItemToModify="((Item=(Type=Bone,Name=\"braidedRopeFront2\"),Transform=(Rotation=(X=-0.0037,Y=-0.0354,Z=-0.103,W=-0.994),Translation=(X=0,Y=0,Z=-0.7917),Scale3D=(X=1,Y=1,Z=1))),(Item=(Type=Bone,Name=\"braidedRopeFront3\"),Transform=(Rotation=(X=0,Y=0,Z=0,W=-1),Translation=(X ...(truncated)"
    WeightMinimum=0
    Weight=@n12.OutputParam
   )
   > n13 = &RigUnit_SphericalPoseReader(
    ActiveRegionScaleFactors=(PositiveWidth=1,NegativeWidth=1,PositiveHeight=1,NegativeHeight=1)
    ActiveRegionSize=0.05
    Debug=(bDrawDebug=true,bDraw2D=False,bDrawLocalAxes=False,DebugScale=25,DebugSegments=20,DebugThickness=0.25)
    DriverAxis=(X=-1,Y=0,Z=0)
    DriverItem=(Type=Bone,Name="upperarm_r")
    FalloffRegionScaleFactors=(PositiveWidth=1,NegativeWidth=1,PositiveHeight=1,NegativeHeight=1)
    FalloffSize=0.2
    FlipHeightScaling=False
    FlipWidthScaling=False
    OptionalParentItem=(Type=Bone,Name="clavicle_r")
    RotationOffset=(X=-47.0916,Y=78.5773,Z=15.3203)
   )
   > &RigUnit_ModifyTransforms(
    ItemToModify="((Item=(Type=Bone,Name=\"braidedRopeFront2\"),Transform=(Rotation=(X=0,Y=0,Z=-0.051,W=-0.9987),Translation=(X=0,Y=0,Z=0),Scale3D=(X=1,Y=1,Z=1))),(Item=(Type=Bone,Name=\"braidedRopeFront3\"),Transform=(Rotation=(X=0,Y=0,Z=0,W=1),Translation=(X=0,Y=2.6039,Z=-1.2 ...(truncated)"
    WeightMinimum=0
    Weight=@n13.OutputParam
   )
   > n14 = &RigUnit_SphericalPoseReader(
    ActiveRegionScaleFactors=(PositiveWidth=1,NegativeWidth=1,PositiveHeight=1,NegativeHeight=1)
    ActiveRegionSize=0.05
    Debug=(bDrawDebug=true,bDraw2D=False,bDrawLocalAxes=False,DebugScale=25,DebugSegments=20,DebugThickness=0.25)
    DriverAxis=(X=-1,Y=0,Z=0)
    DriverItem=(Type=Bone,Name="upperarm_r")
    FalloffRegionScaleFactors=(PositiveWidth=1,NegativeWidth=1,PositiveHeight=1,NegativeHeight=1)
    FalloffSize=0.15
    FlipHeightScaling=False
    FlipWidthScaling=False
    OptionalParentItem=(Type=Bone,Name="clavicle_r")
    RotationOffset=(X=6.785,Y=37.8023,Z=23.25)
   )
   > &RigUnit_ModifyTransforms(
    ItemToModify="((Item=(Type=Bone,Name=\"CordsBot\"),Transform=(Rotation=(X=-0.0941,Y=0.0464,Z=-0.0192,W=-0.9943),Translation=(X=2.2281,Y=-1.0679,Z=-3.2551),Scale3D=(X=1,Y=1,Z=1))),(Item=(Type=Bone,Name=\"braidedRopeBack4\"),Transform=(Rotation=(X=0,Y=0,Z=0,W=-1),Translation= ...(truncated)"
    WeightMinimum=0
    Weight=@n14.OutputParam
   )
   > n15 = &RigUnit_SphericalPoseReader(
    ActiveRegionScaleFactors=(PositiveWidth=1,NegativeWidth=1,PositiveHeight=1,NegativeHeight=1)
    ActiveRegionSize=0.05
    Debug=(bDrawDebug=true,bDraw2D=False,bDrawLocalAxes=False,DebugScale=25,DebugSegments=20,DebugThickness=0.25)
    DriverAxis=(X=-1,Y=0,Z=0)
    DriverItem=(Type=Bone,Name="upperarm_r")
    FalloffRegionScaleFactors=(PositiveWidth=1,NegativeWidth=1,PositiveHeight=1,NegativeHeight=1)
    FalloffSize=0.2
    FlipHeightScaling=False
    FlipWidthScaling=False
    OptionalParentItem=(Type=Bone,Name="clavicle_r")
    RotationOffset=(X=-50.7095,Y=120.4165,Z=31.7064)
   )
   > &RigUnit_ModifyTransforms(
    ItemToModify="((Item=(Type=Bone,Name=\"braidedRopeFront2\"),Transform=(Rotation=(X=0,Y=0,Z=-0.0699,W=-0.9976),Translation=(X=-0.0899,Y=0.6387,Z=-0.7647),Scale3D=(X=1,Y=1,Z=1))),(Item=(Type=Bone,Name=\"braidedRopeFront3\"),Transform=(Rotation=(X=0,Y=0,Z=0,W=-1),Translation=( ...(truncated)"
    WeightMinimum=0
    Weight=@n15.OutputParam
   )
   n16 = &RigVMFunction_MathQuaternionToEuler(
    RotationOrder=ZYX
    Value=&RigUnit_GetRelativeTransformForItem(Child=(Type=Bone,Name="spine_05"), Parent=(Type=Bone,Name="spine_01"), bChildInitial=False, bParentInitial=False).RelativeTransform.Rotation
   )
   &RigUnit_ModifyTransforms(
    ItemToModify="((Item=(Type=Bone,Name=\"CordsBot\"),Transform=(Rotation=(X=-0.3181,Y=0,Z=0,W=0.948),Translation=(X=-0.5232,Y=9.7814,Z=-1.4805),Scale3D=(X=1,Y=1,Z=1))),(Item=(Type=Bone,Name=\"braidedRopeFront5\"),Transform=(Rotation=(X=-0.1027,Y=0,Z=0,W=0.9947),Translation=(X ...(truncated)"
    WeightMinimum=0
    Weight=&RigVMFunction_MathDoubleRemap(SourceMaximum=-70, SourceMinimum=0, TargetMaximum=1, TargetMinimum=0, bClamp=False, Value=@n16.Result.X).Result
    ExecutePin=@RigVMFunction_Sequence_2.B
   )
   > &RigUnit_ModifyTransforms(
    ItemToModify="((Item=(Type=Bone,Name=\"CordsBot\"),Transform=(Rotation=(X=0.3609,Y=0,Z=0,W=0.9326),Translation=(X=0,Y=-6.0876,Z=-5.6223),Scale3D=(X=1,Y=1,Z=1))),(Item=(Type=Bone,Name=\"braidedRopeFront4\"),Transform=(Rotation=(X=-0.0037,Y=0.078,Z=0.0467,W=-0.9959),Translati ...(truncated)"
    WeightMinimum=0
    Weight=&RigVMFunction_MathDoubleRemap(SourceMaximum=50, SourceMinimum=0, TargetMaximum=1, TargetMinimum=0, bClamp=False, Value=@n16.Result.X).Result
   )
   > &RigUnit_ModifyTransforms(
    ItemToModify="((Item=(Type=Bone,Name=\"CordsBot\"),Transform=(Rotation=(X=0,Y=0,Z=0,W=1),Translation=(X=0,Y=-2.4303,Z=0),Scale3D=(X=1,Y=1,Z=1))),(Item=(Type=Bone,Name=\"braidedRopeFront4\"),Transform=(Rotation=(X=0,Y=0,Z=0,W=1),Translation=(X=0,Y=0,Z=0),Scale3D=(X=1,Y=1,Z=1 ...(truncated)"
    WeightMinimum=0
    Weight=&RigVMFunction_MathDoubleRemap(SourceMaximum=-80, SourceMinimum=-10, TargetMaximum=1, TargetMinimum=0, bClamp=False, Value=@n16.Result.Z).Result
   )
   > &RigUnit_ModifyTransforms(
    ItemToModify="((Item=(Type=Bone,Name=\"CordsBot\"),Transform=(Rotation=(X=0,Y=0,Z=0,W=1),Translation=(X=0,Y=6.8098,Z=0),Scale3D=(X=1,Y=1,Z=1))),(Item=(Type=Bone,Name=\"braidedRopeFront4\"),Transform=(Rotation=(X=0,Y=0.1337,Z=0,W=-0.991),Translation=(X=-5.3906,Y=0,Z=0.377),S ...(truncated)"
    WeightMinimum=0
    Weight=&RigVMFunction_MathDoubleRemap(SourceMaximum=100, SourceMinimum=0, TargetMaximum=1, TargetMinimum=0, bClamp=False, Value=@n16.Result.Z).Result
   )
   > &RigUnit_ModifyTransforms(
    ItemToModify="((Item=(Type=Bone,Name=\"CordsBot\"),Transform=(Rotation=(X=0,Y=0.1267,Z=0,W=0.9919),Translation=(X=-2.6,Y=-0.2226,Z=1.0164),Scale3D=(X=1,Y=1,Z=1))),(Item=(Type=Bone,Name=\"braidedRopeFront4\"),Transform=(Rotation=(X=0.0837,Y=0.0138,Z=-0.1619,W=-0.9831),Transl ...(truncated)"
    WeightMinimum=0
    Weight=&RigVMFunction_MathDoubleRemap(SourceMaximum=-40, SourceMinimum=0, TargetMaximum=1, TargetMinimum=0, bClamp=False, Value=@n16.Result.Y).Result
   )
   > &RigUnit_ModifyTransforms(
    ItemToModify=((Item=(Type=Bone,Name="None"),Transform=(Translation=(X=0,Y=0,Z=0),Scale3D=(X=1,Y=1,Z=1))))
    WeightMinimum=0
    Weight=&RigVMFunction_MathDoubleRemap(SourceMaximum=40, SourceMinimum=0, TargetMaximum=1, TargetMinimum=0, bClamp=False, Value=@n16.Result.Y).Result
   )
   n17 = &RigVMFunction_MathQuaternionToEuler(
    RotationOrder=ZYX
    Value=&RigUnit_GetRelativeTransformForItem(Child=(Type=Bone,Name="clavicle_r"), Parent=(Type=Null,Name="clavicle_r_null"), bChildInitial=False, bParentInitial=False).RelativeTransform.Rotation
   )
   n18 = &RigVMFunction_MathDoubleRemap(SourceMaximum=-40, SourceMinimum=0, TargetMaximum=1, TargetMinimum=0, Value=@n17.Result.Y)
   &RigUnit_ModifyTransforms(
    ItemToModify="((Item=(Type=Bone,Name=\"braidedRopeFront3\"),Transform=(Rotation=(X=0,Y=0,Z=-0.1274,W=-0.9919),Translation=(X=-0.1853,Y=0.7095,Z=0),Scale3D=(X=1,Y=1,Z=1))),(Item=(Type=Bone,Name=\"braidedRopeFront4\"),Transform=(Rotation=(X=-0.228,Y=-0.0207,Z=0.0881,W=0.9694) ...(truncated)"
    WeightMinimum=0
    Weight=@n18.Result
    ExecutePin=@RigVMFunction_Sequence_2.A
   )
   n19 = &RigVMFunction_MathDoubleRemap(SourceMaximum=20, SourceMinimum=0, TargetMaximum=1, TargetMinimum=0, Value=@n17.Result.Y)
   > &RigUnit_ModifyTransforms(
    ItemToModify="((Item=(Type=Bone,Name=\"braidedRopeFront2\"),Transform=(Rotation=(X=0,Y=0,Z=0,W=1),Translation=(X=0,Y=0,Z=0),Scale3D=(X=1,Y=1,Z=1))),(Item=(Type=Bone,Name=\"braidedRopeFront5\"),Transform=(Rotation=(X=0,Y=0,Z=0,W=1),Translation=(X=3.231,Y=1.3839,Z=-0.9534),Sc ...(truncated)"
    WeightMinimum=0
    Weight=@n19.Result
   )
   n20 = &RigVMFunction_MathDoubleDiv(B=2, A=@n19.Result)
   Divide = &RigVMFunction_MathDoubleDiv(B=2, A=@n18.Result)
   Subtract = &RigVMFunction_MathDoubleSub(
    B=@Divide.Result
    A=&RigVMFunction_MathDoubleRemap(SourceMaximum=45, SourceMinimum=0, TargetMaximum=1, TargetMinimum=0, Value=@n17.Result.Z).Result
   )
   > &RigUnit_ModifyTransforms(
    ItemToModify="((Item=(Type=Bone,Name=\"braidedRopeFront4\"),Transform=(Rotation=(X=0,Y=0,Z=0,W=1),Translation=(X=0,Y=-6.4527,Z=8.6582),Scale3D=(X=1,Y=1,Z=1))),(Item=(Type=Bone,Name=\"braidedRopeFront5\"),Transform=(Rotation=(X=0,Y=0,Z=0,W=1),Translation=(X=0,Y=4.5123,Z=-4.0 ...(truncated)"
    WeightMinimum=0
    Weight=&RigVMFunction_MathDoubleSub(B=@n20.Result, A=@Subtract.Result).Result
   )
   n21 = &RigVMFunction_MathDoubleRemap(SourceMaximum=-35, SourceMinimum=0, TargetMaximum=1, TargetMinimum=0, Value=@n17.Result.Z)
   > &RigUnit_ModifyTransforms(
    ItemToModify="((Item=(Type=Bone,Name=\"braidedRopeBack2\"),Transform=(Rotation=(X=0.0001,Y=0.1368,Z=-0.015,W=-0.9905),Translation=(X=0,Y=0,Z=0),Scale3D=(X=1,Y=1,Z=1))),(Item=(Type=Bone,Name=\"braidedRopeBack3\"),Transform=(Rotation=(X=-0.2906,Y=0,Z=0,W=0.9568),Translation=( ...(truncated)"
    WeightMinimum=0
    Weight=&RigVMFunction_MathDoubleSub(B=@n20.Result, A=&RigVMFunction_MathDoubleSub(B=@Divide.Result, A=@n21.Result).Result).Result
   )
   > &RigUnit_ModifyTransforms(
    ItemToModify="((Item=(Type=Bone,Name=\"braidedRopeBack2\"),Transform=(Rotation=(X=0,Y=0,Z=0,W=-1),Translation=(X=0,Y=0,Z=0.3846),Scale3D=(X=1,Y=1,Z=1))),(Item=(Type=Bone,Name=\"braidedRopeBack3\"),Transform=(Rotation=(X=0,Y=0,Z=0,W=-1),Translation=(X=-3.0639,Y=0,Z=-0.6867), ...(truncated)"
    WeightMinimum=0
    Weight=&RigVMFunction_MathDoubleRemap(SourceMaximum=-35, SourceMinimum=0, TargetMaximum=1, TargetMinimum=0, Value=@n17.Result.X).Result
   )
   > &RigUnit_ModifyTransforms(
    ItemToModify="((Item=(Type=Bone,Name=\"braidedRopeBack2\"),Transform=(Rotation=(X=0,Y=0.2326,Z=0,W=-0.9726),Translation=(X=-0.453,Y=-0.2341,Z=0.6502),Scale3D=(X=1,Y=1,Z=1))),(Item=(Type=Bone,Name=\"braidedRopeBack3\"),Transform=(Rotation=(X=0,Y=0,Z=0,W=-1),Translation=(X=5. ...(truncated)"
    WeightMinimum=0
    Weight=&RigVMFunction_MathDoubleRemap(SourceMaximum=35, SourceMinimum=0, TargetMaximum=1, TargetMinimum=0, Value=@n17.Result.X).Result
   )
   &RigUnit_ModifyTransforms(
    ExecutePin=()
    ItemToModify="((Item=(Type=Bone,Name=\"CordsBot\"),Transform=(Rotation=(X=-0.3609,Y=0,Z=0,W=-0.9326),Translation=(X=0,Y=-6.0876,Z=-5.6224),Scale3D=(X=1,Y=1,Z=1))),(Item=(Type=Bone,Name=\"braidedRopeFront4\"),Transform=(Rotation=(X=-0.0037,Y=0.078,Z=0.0467,W=-0.9959),Transla ...(truncated)"
    Weight=1
    WeightMinimum=0
   )
   &RigUnit_ModifyTransforms(
    ExecutePin=()
    ItemToModify="((Item=(Type=Bone,Name=\"braidedRopeBack2\"),Transform=(Rotation=(X=0,Y=0.0014,Z=0,W=1),Translation=(X=0,Y=0,Z=0),Scale3D=(X=1,Y=1,Z=1))),(Item=(Type=Bone,Name=\"braidedRopeBack3\"),Transform=(Rotation=(X=-0.2906,Y=0,Z=0,W=0.9568),Translation=(X=0,Y=-7.5906,Z= ...(truncated)"
    Weight=1
    WeightMinimum=0
   )
   &RigUnit_ModifyTransforms(
    ExecutePin=()
    ItemToModify="((Item=(Type=Bone,Name=\"CordsBot\"),Transform=(Rotation=(X=-0.2349,Y=0.0432,Z=-0.0256,W=-0.9707),Translation=(X=2.3206,Y=-0.8718,Z=-4.2714),Scale3D=(X=1,Y=1,Z=1))),(Item=(Type=Bone,Name=\"braidedRopeBack4\"),Transform=(Rotation=(X=0,Y=0,Z=0,W=-1),Translation= ...(truncated)"
    Weight=1
    WeightMinimum=0
   )
   &RigUnit_ModifyTransforms(
    ExecutePin=()
    ItemToModify="((Item=(Type=Bone,Name=\"CordsBot\"),Transform=(Rotation=(X=-0.3181,Y=0,Z=0,W=0.948),Translation=(X=-0.5232,Y=9.7814,Z=-1.4805),Scale3D=(X=1,Y=1,Z=1))),(Item=(Type=Bone,Name=\"braidedRopeFront5\"),Transform=(Rotation=(X=-0.1027,Y=0,Z=0,W=0.9947),Translation=(X ...(truncated)"
    Weight=1
    WeightMinimum=0
   )
   &RigUnit_SphericalPoseReader(
    ActiveRegionScaleFactors=(PositiveWidth=1,NegativeWidth=1,PositiveHeight=1,NegativeHeight=1)
    ActiveRegionSize=0.05
    Debug=(bDrawDebug=true,bDraw2D=False,bDrawLocalAxes=False,DebugScale=25,DebugSegments=20,DebugThickness=0.25)
    DriverAxis=(X=-1,Y=0,Z=0)
    DriverItem=(Type=Bone,Name="upperarm_r")
    ExecutePin=()
    FalloffRegionScaleFactors=(PositiveWidth=1,NegativeWidth=1,PositiveHeight=1,NegativeHeight=1)
    FalloffSize=0.2
    FlipHeightScaling=False
    FlipWidthScaling=False
    OptionalParentItem=(Type=Bone,Name="clavicle_r")
    RotationOffset=(X=-55.8604,Y=-4.7644,Z=6.7182)
   )
   &RigUnit_SphericalPoseReader(
    ActiveRegionScaleFactors=(PositiveWidth=1,NegativeWidth=1.5,PositiveHeight=1,NegativeHeight=1)
    ActiveRegionSize=0.05
    Debug=(bDrawDebug=true,bDraw2D=False,bDrawLocalAxes=False,DebugScale=25,DebugSegments=20,DebugThickness=0.25)
    DriverAxis=(X=-1,Y=0,Z=0)
    DriverItem=(Type=Bone,Name="upperarm_r")
    ExecutePin=()
    FalloffRegionScaleFactors=(PositiveWidth=1,NegativeWidth=1,PositiveHeight=1,NegativeHeight=1)
    FalloffSize=0.2
    FlipHeightScaling=False
    FlipWidthScaling=False
    OptionalParentItem=(Type=Bone,Name="clavicle_r")
    RotationOffset=(X=-34.5829,Y=19.6729,Z=11.4026)
   )
  ]}
]
hierarchy [
 root {loc=<0 0 0>, rot=<0 0 0 1>}
 pelvis {parent=root, loc=<0 -0.7071 93.1191>, rot=<0.0127 0.707 -0.0127 -0.707>}
 spine_01 {parent=pelvis, loc=<6.7494 -0.6388 0>, rot=<0 0 0.034 -0.9994>}
 spine_02 {parent=spine_01, loc=<7.2518 -0.035 0>, rot=<0 0 0.0238 -0.9997>}
 spine_03 {parent=spine_02, loc=<6.0104 0 0>, rot=<0 0 0.0149 -0.9999>}
 spine_04 {parent=spine_03, loc=<8.1024 0 0>, rot=<0 0 -0.0132 -0.9999>}
 spine_05 {parent=spine_04, loc=<10.0146 0 0>, rot=<0 0 -0.0015 -1>}
 neck_01 {parent=spine_05, loc=<11.3138 0.1529 0>, rot=<0 0 -0.2003 -0.9797>}
 neck_02 {parent=neck_01, loc=<4.5627 0 0>, rot=<0 0 0.0504 -0.9987>}
 head {parent=neck_02, loc=<5.3914 0 0>, rot=<0 0 0.115 -0.9934>}
 attach_cape {parent=spine_05, loc=<9.017 -11.714 0>, rot=<0.0616 -0.7044 -0.0616 -0.7044>}
 cape {parent=attach_cape, type=Null, loc=<0 0 0>, rot=<0 0 0 1>}
 attach_backpack {parent=spine_05, loc=<-0.123 -13.473 0>, rot=<-0.0093 -0.707 0.0093 -0.707>}
 Backpack_BR {parent=attach_backpack, type=Null, loc=<0 0 0>, rot=<0 0 0 1>}
 clavicle_l {parent=spine_05, loc=<4.3575 -0.1466 -2.3223>, rot=<0.0465 -0.7289 -0.009 -0.6829>}
 upperarm_l {parent=clavicle_l, loc=<15.5111 -0.0001 0>, rot=<0.0226 -0.3234 -0.0088 -0.9459>}
 lowerarm_l {parent=upperarm_l, loc=<26.6514 0 0>, rot=<0.0391 -0.0099 -0.2453 -0.9686>}
 hand_l {parent=lowerarm_l, loc=<22.3491 0 0>, rot=<-0.7081 -0.0507 0.0975 -0.6975>}
 weapon_l {parent=hand_l, loc=<-0.1974 -3.4111 0.3811>, rot=<0 0 0 1>}
 thumb_01_l {parent=hand_l, loc=<1.3118 -0.3191 -2.459>, rot=<0.6546 -0.4042 0.0664 -0.6354>}
 thumb_02_l {parent=thumb_01_l, loc=<5.9336 -0.0448 0.0401>, rot=<-0.0037 0.0213 0.298 -0.9543>}
 thumb_03_l {parent=thumb_02_l, loc=<2.2223 0.3682 -0.0907>, rot=<0.0118 -0.013 0.0066 -0.9998>}
 thumb_Sleeve_l {parent=thumb_01_l, loc=<-0.0001 0.0001 0>, rot=<0 0 0 1>}
 index_metacarpal_l {parent=hand_l, loc=<1.6535 0.0183 -2.96>, rot=<0.0158 -0.0124 -0.0173 -0.9996>}
 index_01_l {parent=index_metacarpal_l, loc=<8.3008 0.0357 -0.0524>, rot=<-0.0044 0.0485 0.0924 -0.9945>}
 index_02_l {parent=index_01_l, loc=<3.3587 -0.3004 0.0452>, rot=<-0.002 0.0027 0.3225 -0.9466>}
 index_03_l {parent=index_02_l, loc=<2.5578 -0.123 -0.0252>, rot=<-0.002 0.0067 0.2558 -0.9667>}
 middle_metacarpal_l {parent=hand_l, loc=<2.1856 -0.1725 -0.5984>, rot=<0.006 0.0168 -0.0329 -0.9993>}
 middle_01_l {parent=middle_metacarpal_l, loc=<7.3275 -0.2447 0.0129>, rot=<-0.0208 0.0042 0.0858 -0.9961>}
 middle_02_l {parent=middle_01_l, loc=<4.0123 0.1401 0.1175>, rot=<-0.0218 0.0258 0.3512 -0.9357>}
 middle_03_l {parent=middle_02_l, loc=<2.6077 0.0534 0.1373>, rot=<0.0082 0.0188 0.1708 -0.9851>}
 ring_metacarpal_l {parent=hand_l, loc=<2.2074 -0.3625 1.3638>, rot=<-0.0284 0.0722 -0.0458 -0.9959>}
 ring_01_l {parent=ring_metacarpal_l, loc=<7.0295 -0.6197 -0.1446>, rot=<-0.0489 -0.0252 0.1892 -0.9804>}
 ring_02_l {parent=ring_01_l, loc=<3.6174 0.4647 -0.1027>, rot=<0 0 0.2806 -0.9598>}
 ring_03_l {parent=ring_02_l, loc=<2.4152 0.114 0.0199>, rot=<0 0 0.1189 -0.9929>}
 pinky_metacarpal_l {parent=hand_l, loc=<1.7616 -0.9753 3.3814>, rot=<-0.0397 0.0814 -0.069 -0.9935>}
 pinky_01_l {parent=pinky_metacarpal_l, loc=<7.1709 -0.4831 -0.1094>, rot=<-0.0654 -0.0035 0.2242 -0.9723>}
 pinky_02_l {parent=pinky_01_l, loc=<2.8274 0.0893 -0.1043>, rot=<0.0037 0.0011 0.2854 -0.9584>}
 pinky_03_l {parent=pinky_02_l, loc=<2.2046 0.0307 -0.0078>, rot=<-0.0002 0.0098 0.1445 -0.9895>}
 hand_attach_l {parent=hand_l, loc=<8.5 -3 0>, rot=<0.7071 0.7071 0 0>}
 sleeve_l {parent=hand_l, loc=<-0.0004 0.0003 -0.0001>, rot=<0 0 0 1>}
 button_hand_l {parent=sleeve_l, loc=<-0.6716 -1.9999 5.1545>, rot=<0 0 0 1>}
 deform_wrist_up_base_l {parent=hand_l, loc=<0 2.2682 0>, rot=<0 0 0 -1>}
 deform_wrist_up_l {parent=deform_wrist_up_base_l, loc=<0 0 0>, rot=<0 0 0 1>}
 deform_wrist_dn_base_l {parent=hand_l, loc=<0 -2.2342 0>, rot=<0 0 0 -1>}
 deform_wrist_dn_l {parent=deform_wrist_dn_base_l, loc=<0 0 0>, rot=<0 0 0 1>}
 lowerarm_twist_02_l {parent=lowerarm_l, loc=<9.6199 0 0>, rot=<0 0 0 -1>}
 lowerarm_twist_01_l {parent=lowerarm_l, loc=<17.8549 0 0>, rot=<0 0 0 -1>}
 deform_elbow_in_base_l {parent=lowerarm_l, loc=<0.2112 5.259 0.8274>, rot=<0 0 0 -1>}
 deform_elbow_in_l {parent=deform_elbow_in_base_l, loc=<0 0 0>, rot=<0 0 0 1>}
 deform_elbow_out_base_l {parent=lowerarm_l, loc=<0 -4 0>, rot=<0 0 0 -1>}
 deform_elbow_out_l {parent=deform_elbow_out_base_l, loc=<0 0 0>, rot=<0 0 0 1>}
 hand_l_null {parent=lowerarm_l, type=Null, loc=<22.3491 0 0>, rot=<-0.7081 -0.0507 0.0975 -0.6975>}
 upperarm_twist_01_l {parent=upperarm_l, loc=<0.0041 0 0>, rot=<0 0 0 -1>}
 deform_upperarm_bk_base_l {parent=upperarm_twist_01_l, loc=<1.0813 -7.2192 -1.9453>, rot=<0 0 0 1>}
 deform_upperarm_bk_l {parent=deform_upperarm_bk_base_l, loc=<0 0 0>, rot=<0 0 0 1>}
 deform_upperarm_fr_base_l {parent=upperarm_twist_01_l, loc=<-0.3691 7.0507 -2.765>, rot=<0 0 0 1>}
 deform_upperarm_fr_l {parent=deform_upperarm_fr_base_l, loc=<0 0 0>, rot=<0 0 0 1>}
 deform_upperarm_up_base_l {parent=upperarm_twist_01_l, loc=<-3.2216 -1.1391 3.3311>, rot=<0 0 0 1>}
 deform_upperarm_up_l {parent=deform_upperarm_up_base_l, loc=<0 0 0>, rot=<0 0 0 1>}
 VollumeCompensator_l {parent=upperarm_twist_01_l, loc=<0 0 0>, rot=<0 0 0 1>}
 upperarm_twist_02_l {parent=upperarm_l, loc=<13.7941 0 0>, rot=<0 0 0 -1>}
 button_clavicle_l {parent=clavicle_l, loc=<8.2366 -2.1412 8.5231>, rot=<0 0 0 1>}
 clavicle_r {parent=spine_05, loc=<4.3575 -0.1466 2.3223>, rot=<0.7289 0.0465 -0.6829 0.009>}
 upperarm_r {parent=clavicle_r, loc=<-15.5111 0.0001 0>, rot=<0.0226 -0.3234 -0.0088 -0.9459>}
 lowerarm_r {parent=upperarm_r, loc=<-26.6514 0 0>, rot=<0.0391 -0.0099 -0.2453 -0.9686>}
 hand_r {parent=lowerarm_r, loc=<-22.3491 0 0>, rot=<-0.7081 -0.0507 0.0975 -0.6975>}
 weapon_r {parent=hand_r, loc=<0.1974 3.4111 -0.3811>, rot=<0 0 0 1>}
 thumb_01_r {parent=hand_r, loc=<-1.3118 0.3191 2.459>, rot=<0.6546 -0.4042 0.0664 -0.6354>}
 thumb_02_r {parent=thumb_01_r, loc=<-5.9336 0.0448 -0.0401>, rot=<-0.0037 0.0213 0.298 -0.9543>}
 thumb_03_r {parent=thumb_02_r, loc=<-2.2223 -0.3682 0.0907>, rot=<0.0118 -0.013 0.0066 -0.9998>}
 thumb_Sleeve_r {parent=thumb_01_r, loc=<0 0 0>, rot=<0 0 0 1>}
 index_metacarpal_r {parent=hand_r, loc=<-1.6535 -0.0183 2.96>, rot=<0.0158 -0.0124 -0.0173 -0.9996>}
 index_01_r {parent=index_metacarpal_r, loc=<-8.3008 -0.0357 0.0524>, rot=<-0.0044 0.0485 0.0924 -0.9945>}
 index_02_r {parent=index_01_r, loc=<-3.3587 0.3004 -0.0452>, rot=<-0.002 0.0027 0.3225 -0.9466>}
 index_03_r {parent=index_02_r, loc=<-2.5578 0.123 0.0252>, rot=<-0.002 0.0067 0.2558 -0.9667>}
 middle_metacarpal_r {parent=hand_r, loc=<-2.1856 0.1725 0.5984>, rot=<0.006 0.0168 -0.0329 -0.9993>}
 middle_01_r {parent=middle_metacarpal_r, loc=<-7.3275 0.2447 -0.0129>, rot=<-0.0208 0.0042 0.0858 -0.9961>}
 middle_02_r {parent=middle_01_r, loc=<-4.0123 -0.1401 -0.1175>, rot=<-0.0218 0.0258 0.3512 -0.9357>}
 middle_03_r {parent=middle_02_r, loc=<-2.6077 -0.0534 -0.1373>, rot=<0.0082 0.0188 0.1708 -0.9851>}
 ring_metacarpal_r {parent=hand_r, loc=<-2.2074 0.3625 -1.3638>, rot=<-0.0284 0.0722 -0.0458 -0.9959>}
 ring_01_r {parent=ring_metacarpal_r, loc=<-7.0295 0.6197 0.1446>, rot=<-0.0489 -0.0252 0.1892 -0.9804>}
 ring_02_r {parent=ring_01_r, loc=<-3.6174 -0.4647 0.1027>, rot=<0 0 0.2806 -0.9598>}
 ring_03_r {parent=ring_02_r, loc=<-2.4152 -0.114 -0.0199>, rot=<0 0 0.1189 -0.9929>}
 pinky_metacarpal_r {parent=hand_r, loc=<-1.7616 0.9753 -3.3814>, rot=<-0.0397 0.0814 -0.069 -0.9935>}
 pinky_01_r {parent=pinky_metacarpal_r, loc=<-7.1709 0.4831 0.1094>, rot=<-0.0654 -0.0035 0.2242 -0.9723>}
 pinky_02_r {parent=pinky_01_r, loc=<-2.8274 -0.0893 0.1043>, rot=<0.0037 0.0011 0.2854 -0.9584>}
 pinky_03_r {parent=pinky_02_r, loc=<-2.2046 -0.0307 0.0078>, rot=<-0.0002 0.0098 0.1445 -0.9895>}
 hand_attach_r {parent=hand_r, loc=<-8.5001 2.9999 0>, rot=<0 0 -0.7071 -0.7071>}
 sleeve_r {parent=hand_r, loc=<0 0 0>, rot=<0 0 0 1>}
 button_hand_r {parent=sleeve_r, loc=<0.6718 1.9997 -5.1545>, rot=<0 0 0 1>}
 deform_wrist_up_base_r {parent=hand_r, loc=<0.0003 -2.2684 0.0001>, rot=<0 0 0 -1>}
 deform_wrist_up_r {parent=deform_wrist_up_base_r, loc=<0 0 0>, rot=<0 0 0 1>}
 deform_wrist_dn_base_r {parent=hand_r, loc=<-0.0004 2.2345 -0.0001>, rot=<0 0 0 -1>}
 deform_wrist_dn_r {parent=deform_wrist_dn_base_r, loc=<0 0 0>, rot=<0 0 0 1>}
 lowerarm_twist_02_r {parent=lowerarm_r, loc=<-9.6199 0 0>, rot=<0 0 0 -1>}
 lowerarm_twist_01_r {parent=lowerarm_r, loc=<-17.8549 0 0>, rot=<0 0 0 -1>}
 deform_elbow_in_base_r {parent=lowerarm_r, loc=<-0.2111 -5.2591 -0.8275>, rot=<0 0 0 -1>}
 deform_elbow_in_r {parent=deform_elbow_in_base_r, loc=<-0.0001 0 0.0002>, rot=<0 0 0 1>}
 deform_elbow_out_base_r {parent=lowerarm_r, loc=<0 4 -0.0001>, rot=<0 0 0 -1>}
 deform_elbow_out_r {parent=deform_elbow_out_base_r, loc=<0 0 0>, rot=<0 0 0 1>}
 hand_r_null {parent=lowerarm_r, type=Null, loc=<-22.3491 0 0>, rot=<-0.7081 -0.0507 0.0975 -0.6975>}
 upperarm_twist_01_r {parent=upperarm_r, loc=<-0.0041 0 0>, rot=<0 0 0 -1>}
 deform_upperarm_bk_base_r {parent=upperarm_twist_01_r, loc=<-1.0814 7.2192 1.9455>, rot=<0 0 0 1>}
 deform_upperarm_bk_r {parent=deform_upperarm_bk_base_r, loc=<0 0 0>, rot=<0 0 0 1>}
 deform_upperarm_fr_base_r {parent=upperarm_twist_01_r, loc=<0.3689 -7.0507 2.7652>, rot=<0 0 0 1>}
 deform_upperarm_fr_r {parent=deform_upperarm_fr_base_r, loc=<0 0 0>, rot=<0 0 0 1>}
 deform_upperarm_up_base_r {parent=upperarm_twist_01_r, loc=<3.2214 1.1391 -3.3308>, rot=<0 0 0 1>}
 deform_upperarm_up_r {parent=deform_upperarm_up_base_r, loc=<0 0 0>, rot=<0 0 0 -1>}
 VollumeCompensator_r {parent=upperarm_twist_01_r, loc=<0.0041 0 0>, rot=<0 0 0 -1>}
 upperarm_twist_02_r {parent=upperarm_r, loc=<-13.7941 0 0>, rot=<0 0 0 -1>}
 button_clavicle_r {parent=clavicle_r, loc=<-8.2366 2.1412 -8.5231>, rot=<0 0 0 1>}
 CordsRoot {parent=clavicle_r, loc=<-15.3664 1.7089 -6.2004>, rot=<0 0 0 1>}
 braidedRopeFront1 {parent=CordsRoot, loc=<1.0913 -1.2888 -0.4837>, rot=<0.126 0.1775 0.6934 -0.6868>}
 braidedRopeFront2 {parent=braidedRopeFront1, loc=<5.4703 0 0>, rot=<0 0.2496 0 -0.9684>}
 braidedRopeFront3 {parent=braidedRopeFront2, loc=<5.8774 0 0>, rot=<0.032 0.2493 -0.0082 -0.9679>}
 cord_A {parent=CordsRoot, loc=<2.157 -2.7192 -0.9692>, rot=<0.7738 -0.6245 0.1033 -0.0237>}
 cordMetal_A {parent=cord_A, loc=<3.5173 0 0>, rot=<0 -0.3529 0 -0.9356>}
 cordMetal_A_end {parent=cordMetal_A, loc=<8.8205 0 0>, rot=<0 0 0 1>}
 cord_B {parent=CordsRoot, loc=<1.6284 -2.8147 -0.1746>, rot=<0.7579 -0.6244 0.156 0.1067>}
 cord_B2 {parent=cord_B, loc=<3.9956 0 0>, rot=<0 -0.2396 0 -0.9709>}
 cord_B3 {parent=cord_B2, loc=<5.5767 0 0>, rot=<-0.0954 -0.1176 -0.0113 -0.9884>}
 cord_B4 {parent=cord_B3, loc=<7.4209 0 0>, rot=<0 0 0 1>}
 braidedRopeBack1 {parent=CordsRoot, loc=<0.8422 1.2976 -0.3922>, rot=<-0.6509 -0.6993 -0.2306 0.1849>}
 braidedRopeBack2 {parent=braidedRopeBack1, loc=<5.8224 0 0>, rot=<0 -0.3878 0 -0.9218>}
 CordRingFront1 {parent=CordsRoot, loc=<-0.8494 -1.3855 0.0317>, rot=<0.1454 0.1643 0.6858 -0.6939>}
 CordRingFront2 {parent=CordRingFront1, loc=<5.1028 0 0>, rot=<0 0.2617 0 -0.9651>}
 CordRingFront3 {parent=CordRingFront2, loc=<5.3546 0 0>, rot=<0.9749 -0.0187 0.2029 -0.0899>}
 CordRingBack1 {parent=CordsRoot, loc=<-0.9157 1.0888 0.1352>, rot=<-0.6479 -0.6796 -0.2575 0.2284>}
 CordRingBack2 {parent=CordRingBack1, loc=<7.0812 0 0>, rot=<0 -0.3601 0 -0.9329>}
 Cord_D1 {parent=CordsRoot, loc=<-0.5484 1.9294 -0.369>, rot=<0.3652 -0.2181 0.779 0.4607>}
 Cord_D2 {parent=Cord_D1, loc=<7.2931 0 0>, rot=<0 0.2045 0 -0.9789>}
 Cord_D3 {parent=Cord_D2, loc=<4.9794 0 0>, rot=<0 0 0 1>}
 cord_C {parent=CordsRoot, loc=<-1.3543 -5.0399 0.3912>, rot=<0.3115 0.1495 0.7168 -0.6056>}
 cordMetal_C {parent=cord_C, loc=<3.8835 0 0>, rot=<0 0.1271 0 -0.9919>}
 cordMetal_C_end {parent=cordMetal_C, loc=<8.4658 0 0>, rot=<0 0 0 1>}
 braidedRopeFront4 {parent=spine_05, loc=<-5.7255 7.4198 17.2211>, rot=<0.0462 -0.7736 -0.6301 0.0492>}
 CordRingFront4 {parent=braidedRopeFront4, loc=<0.1206 1.8273 -0.6597>, rot=<0.992 0.0048 0.0091 0.1261>}
 braidedRopeBack3 {parent=spine_05, loc=<-2.1022 -10.9789 17.2504>, rot=<0.0035 -0.7136 -0.7005 0.0103>}
 CordRingBack3 {parent=braidedRopeBack3, loc=<2.0857 1.2018 -0.1876>, rot=<0.9878 0.0113 -0.0258 0.1529>}
 clavicle_r_null {parent=spine_05, type=Null, loc=<4.3575 -0.1466 2.3223>, rot=<0.7289 0.0465 -0.6829 0.009>}
 clavicle_l_null {parent=spine_05, type=Null, loc=<4.3575 -0.1466 -2.3223>, rot=<0.0465 -0.7289 -0.009 -0.6829>}
 braidedRopeFront5 {parent=spine_04, loc=<-4.2711 6.2261 17.3755>, rot=<0.1407 -0.7074 0.6805 -0.1295>}
 CordRingFront5 {parent=braidedRopeFront5, loc=<0 -1.9951 0.046>, rot=<-0.0102 0.0002 0.0067 -0.9999>}
 braidedRopeBack4 {parent=spine_04, loc=<-4.0024 -11.2524 17.3676>, rot=<-0.0785 -0.7222 -0.6829 -0.0767>}
 CordRingBack4 {parent=braidedRopeBack4, loc=<0.4283 1.3252 -0.9407>, rot=<0.995 0.0311 0.0702 0.0635>}
 braidedRopeFront6 {parent=spine_02, loc=<0.7416 2.3437 17.4564>, rot=<0.2442 -0.6869 0.6423 -0.2368>}
 CordRingFront6 {parent=braidedRopeFront6, loc=<0 -2.1264 0>, rot=<0.9986 -0.0182 0.0403 -0.0301>}
 braidedRopeFront7 {parent=spine_02, loc=<-2.7714 -0.5796 17.403>, rot=<0.2437 -0.6856 0.6437 -0.2373>}
 CordRingFront7 {parent=braidedRopeFront7, loc=<-0.1235 -2.2979 0.3638>, rot=<0.9986 -0.0181 0.0403 -0.028>}
 braidedRopeBack5 {parent=spine_02, loc=<-0.5557 -9.0672 17.3279>, rot=<-0.2802 -0.6577 -0.6378 -0.2866>}
 CordRingBack5 {parent=braidedRopeBack5, loc=<0.4501 2.0723 0.3863>, rot=<0.997 0.0386 0.0664 -0.0007>}
 braidedRopeBack6 {parent=spine_02, loc=<-4.438 -4.867 17.2155>, rot=<-0.282 -0.6617 -0.6336 -0.2849>}
 CordRingBack6 {parent=braidedRopeBack6, loc=<-0.3084 2.4712 1.0203>, rot=<0.997 0.039 0.0661 0.0056>}
 CordsBot {parent=spine_02, loc=<0 -3.7546 17.7918>, rot=<-0.0128 0 0 -0.9999>}
 thigh_l {parent=pelvis, loc=<-2.5693 -0.0109 -7.7501>, rot=<0.0012 0.0572 0.0155 -0.9982>}
 calf_l {parent=thigh_l, loc=<-39.0484 0 0>, rot=<0.0006 -0.0161 -0.0391 -0.9991>}
 foot_l {parent=calf_l, loc=<-42.561 0.1077 0.4292>, rot=<0.0418 -0.0455 0.0343 -0.9975>}
 ball_l {parent=foot_l, loc=<-6.1599 13.8264 0.6806>, rot=<0.0001 -0.0001 -0.7063 -0.7079>}
 calf_twist_02_l {parent=calf_l, loc=<-18.8064 -0.0967 0.0397>, rot=<0.0237 0.0095 -0.0035 -0.9997>}
 calf_twist_01_l {parent=calf_l, loc=<-35.813 -0.0967 0.0397>, rot=<0.0237 0.0095 -0.0035 -0.9997>}
 deform_knee_in_base_l {parent=calf_l, loc=<1.298 -5.2957 -0.0496>, rot=<0 0 0 -1>}
 deform_knee_in_l {parent=deform_knee_in_base_l, loc=<0 0 0>, rot=<0 0 0 1>}
 deform_knee_out_base_l {parent=calf_l, loc=<0 4 0>, rot=<0 0 0 -1>}
 deform_knee_out_l {parent=deform_knee_out_base_l, loc=<0 0 0>, rot=<0 0 0 1>}
 thigh_twist_01_l {parent=thigh_l, loc=<-21.6303 0 0>, rot=<0.0013 0 0 -1>}
 deform_groin_base_l {parent=thigh_l, loc=<-2.1908 8.1142 -1.7442>, rot=<0 0 0 -1>}
 deform_groin_l {parent=deform_groin_base_l, loc=<0 0 0>, rot=<0 0 0 1>}
 thigh_r {parent=pelvis, loc=<-2.5693 -0.0109 7.7501>, rot=<0.0572 -0.0012 0.9982 0.0155>}
 calf_r {parent=thigh_r, loc=<39.0484 0 0>, rot=<0.0006 -0.0161 -0.0391 -0.9991>}
 foot_r {parent=calf_r, loc=<42.561 -0.1077 -0.4292>, rot=<0.0418 -0.0455 0.0343 -0.9975>}
 ball_r {parent=foot_r, loc=<6.1599 -13.8264 -0.6806>, rot=<0.0001 -0.0001 -0.7063 -0.7079>}
 calf_twist_02_r {parent=calf_r, loc=<18.8064 0.0967 -0.0397>, rot=<0.0237 0.0095 -0.0035 -0.9997>}
 calf_twist_01_r {parent=calf_r, loc=<35.813 0.0967 -0.0397>, rot=<0.0237 0.0095 -0.0035 -0.9997>}
 deform_knee_in_base_r {parent=calf_r, loc=<-1.298 5.2957 0.0496>, rot=<0 0 0 -1>}
 deform_knee_in_r {parent=deform_knee_in_base_r, loc=<0 0 0>, rot=<0 0 0 1>}
 deform_knee_out_base_r {parent=calf_r, loc=<0 -4 0>, rot=<0 0 0 -1>}
 deform_knee_out_r {parent=deform_knee_out_base_r, loc=<0 0 0>, rot=<0 0 0 1>}
 thigh_twist_01_r {parent=thigh_r, loc=<21.6303 0 0>, rot=<0.0013 0 0 -1>}
 deform_groin_base_r {parent=thigh_r, loc=<2.1908 -8.1142 1.7442>, rot=<0 0 0 -1>}
 deform_groin_r {parent=deform_groin_base_r, loc=<0 0 0>, rot=<0 0 0 1>}
 deform_glute_base_l {parent=pelvis, loc=<-8.8064 -8.8222 -8.4287>, rot=<0.0275 0.0501 -0.4472 -0.8926>}
 deform_glute_l {parent=deform_glute_base_l, loc=<0 0 0>, rot=<0 0 0 1>}
 deform_glute_base_r {parent=pelvis, loc=<-8.8062 -8.8222 8.4287>, rot=<0.0501 -0.0275 0.8926 -0.4472>}
 deform_glute_r {parent=deform_glute_base_r, loc=<0 0 0>, rot=<0 0 0 1>}
 deform_hip_r_base {parent=pelvis, loc=<-2.4197 0.1918 15.1259>, rot=<0.008 0.6262 0.0143 -0.7795>}
 deform_hip_r {parent=deform_hip_r_base, loc=<0 0 0>, rot=<0 0 0 1>}
 deform_hip_l_base {parent=pelvis, loc=<-2.4197 0.1918 -15.1259>, rot=<0.6262 -0.008 0.7795 0.0143>}
 deform_hip_l {parent=deform_hip_l_base, loc=<0 0 0>, rot=<0 0 0 1>}
 Jacket_pelvic_Deformer {parent=pelvis, loc=<-5.6997 12.013 0>, rot=<-0.0127 -0.707 0.0127 -0.707>}
 ik_hand_root {parent=root, loc=<0 0 0>, rot=<0 0 0 1>}
 ik_hand_gun {parent=ik_hand_root, loc=<-51.9727 4.0883 102.6935>, rot=<0.6084 0.1793 0.4613 -0.6205>}
 ik_hand_r {parent=ik_hand_gun, loc=<0 0 0>, rot=<0 0 0 1>}
 ik_hand_l {parent=ik_hand_gun, loc=<53.0285 82.1781 35.2051>, rot=<-0.5102 -0.7906 -0.3387 0>}
 spine_01_ctrl {controlType=EULER_TRANSFORM, type=Control, loc=<0 0 0>, rot=<0 0 0 1>}
 spine_02_ctrl {controlType=EULER_TRANSFORM, parent=spine_01_ctrl, type=Control, loc=<0 0 0>, rot=<0 0 0 1>}
 spine_03_ctrl {controlType=EULER_TRANSFORM, parent=spine_02_ctrl, type=Control, loc=<0 0 0>, rot=<0 0 0 1>}
 spine_04_ctrl {controlType=EULER_TRANSFORM, parent=spine_03_ctrl, type=Control, loc=<0 0 0>, rot=<0 0 0 1>}
 spine_05_ctrl {controlType=EULER_TRANSFORM, parent=spine_04_ctrl, type=Control, loc=<0 0 0>, rot=<0 0 0 1>}
 clavicle_r_ctrl {controlType=EULER_TRANSFORM, parent=spine_05_ctrl, type=Control, loc=<0 0 0>, rot=<0 0 0 1>}
 upperarm_r_ctrl {controlType=EULER_TRANSFORM, parent=clavicle_r_ctrl, type=Control, loc=<0 0 0>, rot=<0 0 0 1>}
 hand_r_ctrl {controlType=EULER_TRANSFORM, parent=upperarm_r_ctrl, type=Control, loc=<0 0 0>, rot=<0 0 0 1>}
 hand_l_ctrl {controlType=EULER_TRANSFORM, type=Control, loc=<0 0 0>, rot=<0 0 0 1>}
]
variables [UpperArm_front {type=float, default=0}]
