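# Physics asset for the braided rope and dangling cords on the CanineCronutDig medium male player body cosmetic.
# Front and back rope chains run from sphere roots through capsules, joining at their last bones, with limits generally loosening along each chain.
# Four short cord chains, two with metal ends, use swing-limited, zero-twist joints.
# Unconstrained pelvis, spine, right clavicle and right upper arm capsules supply body shapes, likely for collision.
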PhysicsAsset /BRCosmetics/Characters/Player/Male/Medium/Bodies/M_MED_CanineCronutDig/Meshes/M_MED_CanineCronutDig_Physics_PartsCords
bodies {
 Cord_D1 [{shapeType=Sphere, radius=1.4153, shapeName=braidedRopeBack1_sphere, center=<1.3139 -0.096 0.3742>}]
 Cord_D2 [{
   shapeType=Capsule
   length=4.1457
   radius=2.0359
   shapeName=braidedRopeBack2_capsule
   center=<0.9537 -0.0929 -0.0504>
   rotation=<rot -0.8235 83.1209 83.2065>
  }]
 Cord_D3 [{shapeType=Capsule, length=3.6306, radius=1.4484, shapeName=braidedRopeBack2_capsule, center=<2.811 0 0>, rotation=<rot 0 90 95.0289>}]
 braidedRopeBack1 [{shapeType=Sphere, radius=1.4153, shapeName=braidedRopeBack1_sphere, center=<1.3139 -0.096 0.3742>}]
 braidedRopeBack2 [{
   shapeType=Capsule
   length=4.4624
   radius=2.9471
   shapeName=braidedRopeBack2_capsule
   center=<1.6769 -0.0949 0.788>
   rotation=<rot 0 90.6509 -77.9021>
  }]
 braidedRopeBack3 [{
   shapeType=Capsule
   length=6.5755
   radius=3.1028
   shapeName=braidedRopeBack3_capsule
   center=<0.8025 0.0305 -0.2654>
   rotation=<rot 0 -91.8477 -88.4988>
  }]
 braidedRopeBack4 [{
   shapeType=Capsule
   length=4.0372
   radius=2.2371
   shapeName=braidedRopeBack4_capsule
   center=<-0.5142 -0.0747 0.3389>
   rotation=<rot 0 -90.988 -78.45>
  }]
 braidedRopeBack5 [{
   shapeType=Capsule
   length=3.9204
   radius=2.1607
   shapeName=braidedRopeBack5_capsule
   center=<-2.4831 -0.0579 1.5003>
   rotation=<rot 0 -89.6243 -59.7188>
  }]
 braidedRopeBack6 [{
   shapeType=Capsule
   length=3.8477
   radius=1.7067
   shapeName=braidedRopeBack6_capsule
   center=<-0.7702 0.0578 0.1823>
   rotation=<rot 0 90.5702 -61.9574>
  }]
 braidedRopeFront1 [{shapeType=Sphere, radius=1.5643, shapeName=braidedRopeFront1_sphere, center=<0.8928 -0.2532 0.0247>}]
 braidedRopeFront2 [{
   shapeType=Capsule
   length=2.4207
   radius=1.328
   shapeName=braidedRopeFront2_capsule
   center=<0.4862 -0.0804 -0.009>
   rotation=<rot 0 -91.2678 -77.1882>
  }]
 braidedRopeFront3 [{
   shapeType=Capsule
   length=3.31
   radius=1.3257
   shapeName=braidedRopeFront3_capsule
   center=<0.3332 0.0276 0.1752>
   rotation=<rot 0 -91.862 -77.1681>
  }]
 braidedRopeFront4 [{
   shapeType=Capsule
   length=4.3735
   radius=1.8837
   shapeName=braidedRopeFront4_capsule
   center=<-0.5468 0.0539 -0.3156>
   rotation=<rot 0 89.943 -83.8084>
  }]
 braidedRopeFront5 [{
   shapeType=Capsule
   length=6.0032
   radius=1.763
   shapeName=braidedRopeFront5_capsule
   center=<-0.3639 0.0344 0.1288>
   rotation=<rot 0 -90.5029 -81.5169>
  }]
 braidedRopeFront6 [{
   shapeType=Capsule
   length=5.5051
   radius=1.6186
   shapeName=braidedRopeFront6_capsule
   center=<-1.3849 0.0899 0.4364>
   rotation=<rot 0 -89.0036 -77.747>
  }]
 braidedRopeFront7 [{
   shapeType=Capsule
   length=2.3114
   radius=1.4477
   shapeName=braidedRopeFront7_capsule
   center=<-0.2005 0.0805 0.1605>
   rotation=<rot 0 81.4928 -80.176>
  }]
 clavicle_r [{
   shapeType=Capsule
   length=7.7283
   radius=7.2117
   shapeName=clavicle_r_capsule
   center=<-9.9199 0.4513 0>
   rotation=<rot 83.9496 -0.0004 -0.0005>
  }]
 cordMetal_A [{shapeType=Capsule, length=5.944, radius=1.5602, shapeName=braidedRopeBack2_capsule, center=<4.5929 -0.0929 0.3133>, rotation=<rot 0 90 90>}]
 cordMetal_C [{shapeType=Capsule, length=6.1353, radius=1.369, shapeName=braidedRopeBack2_capsule, center=<3.5154 -0.0929 -0.2157>, rotation=<rot 0 90 90>}]
 cord_A [{shapeType=Sphere, radius=1.4153, shapeName=braidedRopeBack1_sphere, center=<1.3139 -0.096 0.3742>}]
 cord_B [{shapeType=Sphere, radius=1, shapeName=cord_B_sphere}]
 cord_B2 [{
   shapeType=Capsule
   length=2.9981
   radius=1.1055
   shapeName=braidedRopeBack2_capsule
   center=<2.4967 0.0604 0.1593>
   rotation=<rot -0.3663 92.7972 93.4626>
  }]
 cord_B3 [{
   shapeType=Capsule
   length=5.5784
   radius=1.777
   shapeName=braidedRopeBack2_capsule
   center=<4.2206 0.1445 0.135>
   rotation=<rot -0.3663 92.7972 92.625>
  }]
 cord_C [{shapeType=Sphere, radius=1.4153, shapeName=braidedRopeBack1_sphere, center=<1.3139 -0.096 0.3742>}]
 pelvis [{shapeType=Capsule, length=3.5951, radius=13.4397, shapeName=pelvis_capsule}]
 spine_02 [{shapeType=Capsule, length=7.0846, radius=12.1664, shapeName=spine_02_capsule, center=<-2.1954 0 0>}]
 spine_03 [{shapeType=Capsule, length=8.7088, radius=12.0873, shapeName=spine_03_capsule, center=<2.6783 1.9905 0.9806>}]
 spine_04 [{shapeType=Capsule, length=8.9944, radius=14.0729, shapeName=spine_04_capsule}]
 spine_05 [{shapeType=Capsule, length=17.2041, radius=10.9568, shapeName=spine_05_capsule, center=<0 -1.054 0>}]
 upperarm_r [{shapeType=Capsule, length=20.4017, radius=6.621, shapeName=upperarm_r_capsule, center=<-11.8764 -0.0829 1.2344>, rotation=<rot 90 5.5688 0>}]
}
constraints [
 {bone1Name=braidedRopeBack1, bone2Name=braidedRopeBack2, swing1LimitDegrees=5, swing2LimitDegrees=5, twistLimitDegrees=0}
 {bone1Name=braidedRopeBack2, bone2Name=braidedRopeBack3, swing1LimitDegrees=10, swing2LimitDegrees=10, twistLimitDegrees=0}
 {bone1Name=braidedRopeBack3, bone2Name=braidedRopeBack4, swing1LimitDegrees=25, swing2LimitDegrees=25, twistLimitDegrees=10}
 {bone1Name=braidedRopeBack4, bone2Name=braidedRopeBack5, swing1LimitDegrees=10, swing2LimitDegrees=10, twistLimitDegrees=10}
 {bone1Name=braidedRopeBack5, bone2Name=braidedRopeBack6, swing1LimitDegrees=25, swing2LimitDegrees=25, twistLimitDegrees=10}
 {bone1Name=braidedRopeFront1, bone2Name=braidedRopeFront2, swing1LimitDegrees=5, swing2LimitDegrees=5, twistLimitDegrees=0}
 {bone1Name=braidedRopeFront2, bone2Name=braidedRopeFront3, swing1LimitDegrees=5, swing2LimitDegrees=5, twistLimitDegrees=0}
 {bone1Name=braidedRopeFront3, bone2Name=braidedRopeFront4, swing1LimitDegrees=5, swing2LimitDegrees=5, twistLimitDegrees=0}
 {bone1Name=braidedRopeFront4, bone2Name=braidedRopeFront5, swing1LimitDegrees=10, swing2LimitDegrees=10, twistLimitDegrees=0}
 {bone1Name=braidedRopeFront5, bone2Name=braidedRopeFront6, swing1LimitDegrees=10, swing2LimitDegrees=10, twistLimitDegrees=10}
 {bone1Name=braidedRopeFront6, bone2Name=braidedRopeFront7, swing1LimitDegrees=10, swing2LimitDegrees=10, twistLimitDegrees=10}
 {bone1Name=braidedRopeFront7, bone2Name=braidedRopeBack6, swing1LimitDegrees=25, swing2LimitDegrees=25, twistLimitDegrees=10}
 {bone1Name=Cord_D1, bone2Name=Cord_D2, swing1LimitDegrees=15, swing2LimitDegrees=15, twistLimitDegrees=0}
 {bone1Name=Cord_D2, bone2Name=Cord_D3, swing1LimitDegrees=15, swing2LimitDegrees=15, twistLimitDegrees=0}
 {bone1Name=cord_A, bone2Name=cordMetal_A, swing1LimitDegrees=15, swing2LimitDegrees=15, twistLimitDegrees=0}
 {bone1Name=cord_C, bone2Name=cordMetal_C, swing1LimitDegrees=15, swing2LimitDegrees=15, twistLimitDegrees=0}
 {bone1Name=cord_B, bone2Name=cord_B2, swing1LimitDegrees=15, swing2LimitDegrees=15, twistLimitDegrees=0}
 {bone1Name=cord_B2, bone2Name=cord_B3, swing1LimitDegrees=15, swing2LimitDegrees=15, twistLimitDegrees=0}
]
