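# Post-process-style AnimBP for the CanineCronutDig body cosmetic, copying the parent mesh pose and adding secondary physics.
# A Control Rig feeds three chained RigidBody nodes, then a fixed braidedRopeBack2 bone offset.
# Right upper-arm yaw crossfades full and partial cord simulations.
# Emote, tactical sprint, skydive dive or skydive jump drives reloadAlpha, blending in the first simulation and fading both cord ones; its FInterpTo has DeltaTime and InterpSpeed inputs swapped.
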AnimBlueprint /BRCosmetics/Characters/Player/Male/Medium/Bodies/M_MED_CanineCronutDig/Meshes/M_MED_CanineCronutDig_AnimBP
target_skeleton=&/M_MED_CanineCronutDig_Skeleton
graphs [
 AnimGraph {nodes [
   n1 = RigidBody(ComponentPose=Local To Component(LocalPose=Use cached pose 'PreSim'()), Alpha=Get reloadAlpha() -> (reloadAlpha:double)) {
    bOverrideWorldGravity=true
    overridePhysicsAsset=&/M_MED_CanineCronutDig_Physics2
    simulationSpace=BaseBoneSpace
    baseBoneRef {boneName=spine_04}
    componentAppliedLinearAccClamp=<900 900 900>
    overrideWorldGravity {z=-980}
    simSpaceSettings {maxAngularAcceleration=50, maxAngularVelocity=30, maxLinearAcceleration=50, maxLinearVelocity=30, worldAlpha=0.2}
   }
   n2 = RigidBody(ComponentPose=@n1, Alpha=Get AlphaFullDyn() -> (AlphaFullDyn:double)) {
    overridePhysicsAsset=&/M_MED_CanineCronutDig_Physics_FullCords
    simulationSpace=BaseBoneSpace
    baseBoneRef {boneName=spine_05}
    overrideWorldGravity {z=-980}
    simSpaceSettings {maxAngularAcceleration=40, maxAngularVelocity=4, maxLinearAcceleration=95, maxLinearVelocity=250, worldAlpha=1}
   }
   n3 = RigidBody(ComponentPose=@n2, Alpha=Get AlphaPartDyn() -> (AlphaPartDyn:double)) {
    overridePhysicsAsset=&/M_MED_CanineCronutDig_Physics_PartsCords
    simulationSpace=BaseBoneSpace
    baseBoneRef {boneName=spine_05}
    overrideWorldGravity {z=-980}
    simSpaceSettings {maxAngularAcceleration=40, maxAngularVelocity=4, maxLinearAcceleration=95, maxLinearVelocity=250, worldAlpha=0.2}
   }
   n4 = "Transform (Modify) Bone - Bone: braidedRopeBack2"(ComponentPose=@n3, Alpha=1, Translation="5.822,0,0", Rotation="0,0,0", Scale="1,1,1") {translationMode=BMM_Replace, translationSpace=BCS_ParentBoneSpace, boneToModify {boneName=braidedRopeBack2}, translation {x=5.822}}
   Output Pose(Result=Component To Local(ComponentPose=@n4)) {layerGroup=None}
   n5 = Control Rig(Source=Copy Pose From Mesh() {bUseAttachedParent=true}) {controlRigAssetReference {blueprintRigClass=&/M_MED_CanineCronutDig_CtrlRig.M_MED_CanineCronutDig_CtrlRig_C}}
   Save cached pose 'PreSim'(Pose=@n5) {cacheName=PreSim}
  ]}
 EventGraph {nodes [
   n1 = Event BlueprintUpdateAnimation()
   n2 = OR Boolean!(
    A=Get bIsPlayingEmote()
    B=Get bIsTacticalSprinting()
    C:bool=Get bIsSkydiveDiveMode()
    D:bool=AND Boolean(A=Get bIsSkyDiving(), B=Get bIsJumping())
   ) -> (ReturnValue)
   > Set reloadAlpha(reloadAlpha:double=FInterpTo(Current=Get reloadAlpha(), Target="To Float (Boolean)"(InBool=@n2), DeltaTime=6, InterpSpeed=GetWorldDeltaSeconds!() -> (ReturnValue:double))) -> (Output_Get:double)
   n3 = GetOwnerSkeletalMeshComponent()
   n4 = Remap Float(
    Input=GetSocketTransform!(self=@n3, InSocketName=upperarm_r, TransformSpace=RTS_ParentBoneSpace) -> (
     ReturnValue_Location:Vector
     ReturnValue_Rotation_Roll:float
     ReturnValue_Rotation_Pitch:float
     ReturnValue_Rotation_Yaw:float
     ReturnValue_Scale:Vector
    ).ReturnValue_Rotation_Yaw
    OldRange=(X=50,Y=100)
    NewRange=(X=1,Y=0)
    Clamp=true
   ) -> (Output=~)
   n5 = Get AlphaFullDyn()
   > n6 = Set AlphaFullDyn(AlphaFullDyn:double=Lerp(A=FInterpTo(Current=@n5, Target=@n4, DeltaTime=8, InterpSpeed=@n1.DeltaTimeX), B=0, Alpha=Get reloadAlpha())) -> (Output_Get:double)
   > Set AlphaPartDyn(AlphaPartDyn:double=Lerp(A=float - float(A=1, B=@n6), B=0, Alpha=Get reloadAlpha())) -> (Output_Get:double)
   PrintString!(
    execute
    InString="To String (Rotator)"!(InRot:Rotator=GetSocketTransform!(self=@n3, InSocketName=upperarm_r, TransformSpace=RTS_ParentBoneSpace) -> (ReturnValue_Location, ReturnValue_Rotation:Rotator, ReturnValue_Scale).ReturnValue_Rotation) -> (ReturnValue:String)
    bPrintToScreen
    bPrintToLog=false
    TextColor
    Duration
    Key
   ) -> (then)
  ]}
]
variables [
 AlphaFullDyn {type=number}
 AlphaPartDyn {type=number}
 reloadAlpha {type=number}
]
